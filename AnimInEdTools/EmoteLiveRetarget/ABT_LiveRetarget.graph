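# Template Animation Blueprint that live-retargets animation onto skeletal meshes owned by a BP_LiveRetarget actor, with no fixed target skeleton.
# On initialization it finds its owning component's index in the actor's Target Skeletal Meshes array and takes the IKRetargeter at that index in IKRetargets.
# The AnimGraph feeds that asset into a Retarget Pose From Mesh node, so one blueprint serves every target mesh.
# The IKRetargeter Struct variable is unused in these graphs.
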AnimBlueprint /AnimInEdTools/EmoteLiveRetarget/ABT_LiveRetarget
target_skeleton=null
graphs [
 AnimGraph {nodes [Output Pose(Result=Retarget Pose From Mesh!(IKRetargeterAsset:IKRetargeter Object Reference=Get IKRetargeter Asset() -> (IKRetargeter Asset:IKRetargeter Object Reference)) -> (Pose:Pose Link Structure) {is=title AnimGraphNode_RetargetPoseFromMesh})]}
 EventGraph {nodes [
   Event BlueprintInitializeAnimation()
   n1 = Cast To BP_LiveRetarget!(Object=GetOwningActor()) -> (AsBP Live Retarget:BP Live Retarget Object Reference, bSuccess:bool=~) {is=title K2Node_DynamicCast}
   n2 = Find Item!(
    TargetArray:Array of Skeletal Mesh Component Object References=Get Target Skeletal Meshes!(self:BP Live Retarget Object Reference=@n1) -> (Target Skeletal Meshes:Array of Skeletal Mesh Component Object References)
    ItemToFind:"Skeletal Mesh Component Object Reference (by ref)"=GetOwningComponent()
   ) -> (ReturnValue)
   > Branch(Condition="Not Equal (Integer)"(A=@n2, B=-1))
   > Set IKRetargeter Asset(IKRetargeter Asset:IKRetargeter Object Reference="Get (a copy)"(
     Array:Array of IKRetargeter Object References=Get IKRetargets!(self:BP Live Retarget Object Reference=@n1) -> (IKRetargets:Array of IKRetargeter Object References)
     Dimension 1=@n2
    ) -> (Output:IKRetargeter Object Reference)) -> (Output_Get:IKRetargeter Object Reference)
  ]}
]
variables [
 IKRetargeter Asset {type=IKRetargeter, default=null}
 IKRetargeter Struct {
  type=RigUnit_UAFIKRetargeter
  default {
   iKRetargetAsset=null
   customRetargetProfile {
    bApplyChainSettings=false
    bApplyGlobalSettings=false
    bApplyRootSettings=false
    bApplySourceRetargetPose=false
    bApplyTargetRetargetPose=false
    bForceAllIKOff=false
    sourceRetargetPoseName=None
    targetRetargetPoseName=None
    chainSettings {}
    globalSettings {
     bCopyBasePose=false
     bEnableFK=true
     bEnableIK=true
     bEnablePost=true
     bEnableRoot=true
     bWarping=false
     copyBasePoseRoot=None
     directionChain=None
     directionSource=Goals
     forwardDirection=Y
     sidewaysOffset=0
     sourceScaleFactor=1
     warpForwards=1
     warpSplay=1
    }
    retargetOpProfiles []
    rootSettings {
     affectIKHorizontal=1
     affectIKVertical=0
     blendToSource=0
     rotationAlpha=1
     scaleHorizontal=1
     scaleVertical=1
     translationAlpha=1
     blendToSourceWeights=<1 1 1>
     rotationOffset=<rot 0 0 0>
     translationOffset=<0 0 0>
    }
   }
   executeContext {}
   overrideSetNames []
   sourcePose {}
   targetAnimGraphRefPose {}
   targetPose {}
   workData {}
  }
 }
]
