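; BP_LiveRetarget is a skeletal mesh actor in the emote live-retarget tools that assembles an original character and up to ten targets.
; Its construction script passes each character definition (CID) to CreateSKMbasedonCID, or directly assigns companion meshes when no original CID exists.
; That function blocking-loads character parts, attaches a mesh component for each, puts the Body's master skeleton mesh on the parent and applies part AnimBPs.
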(event EventBeginPlay)

(event Collision|EventActorBeginOverlap
  (params (OtherActor :type Actor%20Object%20Reference)))

(event EventTick
  (params (DeltaSeconds :type Float%20%28single-precision%29)))

(fn CreateSKMbasedonCID
  (params (Invis%20Mesh%20Component :type Skeletal%20Mesh%20Component%20Object%20Reference) (CID :type Athena%20Character%20Item%20Definition%20Object%20Reference) (Include%20Face%20Part :type Boolean))
  (bind _current_part (LocalVariables|GetCurrent%20Part))
  (bind _returnvalue (CharacterPart|Animation|GetAnimClass _current_part))
  (if (and (Utilities|IsValid Invis%20Mesh%20Component :!class /Script/Engine.KismetSystemLibrary) (Utilities|IsValid CID :!class /Script/Engine.KismetSystemLibrary))
    (Utilities|String|LogString (Utilities|GetDisplayName CID) false)
    (for _array_element (Variables|/Script/FortniteGame.AthenaCharacterItemDefinition|GetBaseCharacterParts CID)
      (bind _returnvalue_2 (Utilities|LoadAssetBlocking _array_element))
      (bind _ascustom_character_part (Utilities|Casting|CastToCustomCharacterPart _returnvalue_2))
      (Utilities|String|LogString (Utilities|GetDisplayName _ascustom_character_part) false)
      ; Allow face parts to be turned on/off
      (if (Utilities|Enum|Equal(Enum) (CharacterPart|Art|GetPartType _ascustom_character_part) "Face")
        (if (Variables|GetInclude%20Face%20Part)
          _))
      (bind _output_get (LocalVariables|SetCurrent%20Part _ascustom_character_part))
      (if (Utilities|IsValid _output_get :!class /Script/Engine.KismetSystemLibrary)
        ; Add the rest of the character parts
        (bind _returnvalue_3 (AddComponent|Common|AddSkeletalMeshComponent true (Math|Transform|MakeTransform "0, 0, 0")))
        (bind _output_get_2 (LocalVariables|SetCurrent%20SKM%20comp _returnvalue_3))
        (Transformation|AttachComponentToComponent _output_get_2 (Variables|GetInvis%20Mesh%20Component) "None" "KeepRelative" "KeepRelative" "KeepRelative" false)
        (Components|SkeletalMesh|SetSkeletalMeshAsset (LocalVariables|GetCurrent%20SKM%20comp) (CharacterPart|Art|GetSkeletalMesh (LocalVariables|GetCurrent%20Part)))
        (if (Utilities|Enum|Equal(Enum) (CharacterPart|Art|GetCharacterPartType (LocalVariables|GetCurrent%20Part)) "Body")
          (Utilities|FlowControl|Sequence
            (:then_0
              (bind _returnvalue_4 (Utilities|LoadAssetBlocking (Utilities|Array|Get(acopy) (CharacterPart|Art|GetMasterSkeletalMeshes (LocalVariables|GetCurrent%20Part)))))
              (bind _asskeletal_mesh (Utilities|Casting|CastToSkeletalMesh _returnvalue_4))
              (Components|SkeletalMesh|SetSkeletalMeshAsset (Variables|GetInvis%20Mesh%20Component) _asskeletal_mesh))
            (:then_1
              _))
          (else
            _))
        (if (Utilities|IsValidClass _returnvalue)
          (Components|SkeletalMesh|SetAnimInstanceClass (LocalVariables|GetCurrent%20SKM%20comp) _returnvalue)
          (Components|Animation|SetAnimationMode (LocalVariables|GetCurrent%20SKM%20comp) "AnimationBlueprint" false))))
    (return true)
    (else
      (return false))))

(fn ConstructionScript
  (Utilities|Array|Clear (Variables|Default|GetTarget%20Skeletal%20Meshes))
  (Variables|Default|SetTarget%20Skeletal%20Meshes (Utilities|Array|MakeArray (Variables|Default|GetTarget%20A) (Variables|Default|GetTarget%20B) (Variables|Default|GetTarget%20C) (Variables|Default|GetTarget%20D) (Variables|Default|GetTarget%20E) (Variables|Default|GetTarget%20F) (Variables|Default|GetTarget%20G) (Variables|Default|GetTarget%20H) (Variables|Default|GetTarget%20I) (Variables|Default|GetTarget%20J)))
  (Utilities|IsValid (Variables|LiveData|CID|GetOriginal%20CID)
    (:Is%20Valid
      (CallFunction|CreateSKMBasedOnCID (Variables|SkeletalMeshActor|GetSkeletalMeshComponent) (Variables|LiveData|CID|GetOriginal%20CID) (Variables|LiveData|CID|GetParent%20CID%20Show%20Face%20Part))
      (for (_array_element _array_index) (Variables|LiveData|CID|GetTarget%20CIDs)
        (bind _result (ArrayMacros|SafeArrayGet (Variables|Default|GetTarget%20Skeletal%20Meshes) _array_index)
          (:%20%20
            (CallFunction|CreateSKMBasedOnCID _result _array_element (Utilities|Array|Get(acopy) (Variables|LiveData|CID|GetCID%20Show%20Face%20Part) _array_index))))))
    (:Is%20Not%20Valid
      (Components|SkeletalMesh|SetSkeletalMeshAsset (Variables|SkeletalMeshActor|GetSkeletalMeshComponent) (Variables|LiveData|Companions|GetOriginal%20Companion%20Skeletal%20Mesh))
      (for (_array_element_2 _array_index_2) (Variables|LiveData|Companions|GetTarget%20Companion%20Skeletal%20Meshes)
        (Components|SkeletalMesh|SetSkeletalMeshAsset (Utilities|Array|Get(acopy) (Variables|Default|GetTarget%20Skeletal%20Meshes) _array_index_2) _array_element_2)))))
