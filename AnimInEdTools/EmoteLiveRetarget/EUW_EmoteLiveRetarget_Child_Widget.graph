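# Editor Utility child widget for one Emote Live Retarget target, exposing character item, Control Rig, IK Retargeter and linked-anim suffix.
# Item edits fire E_Update_Suffix; CID Changed fills those fields through embedded Python.
# Athena characters get the biped rig, a suffix from body type/gender or montage-table dependencies, and a matching retargeter.
# Companions take the section label's suffix, a name-matched rig and a retargeter found among AnimBP dependencies.
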EditorUtilityWidgetBlueprint /AnimInEdTools/EmoteLiveRetarget/EUW_EmoteLiveRetarget_Child_Widget
graphs [
 E_Update_Suffix {nodes [E_Update_Suffix() {is=title K2Node_FunctionEntry}]}
 EventGraph {nodes [
   "On Property Changed (CIDView)"!() -> (OutputDelegate, then, PropertyName:Name=None) {is=title K2Node_ComponentBoundEvent}
   > Call E Update Suffix!(execute:Exec, self:EUW Emote Live Retarget Child Widget Object Reference) -> (then:Exec) {is=title K2Node_CallDelegate}
   CID Changed() {is=title K2Node_CustomEvent}
   > n1 = Cast To AthenaCharacterItemDefinition!(execute, Object=Get CID() -> (CID:Fort Account Item Definition Object Reference)) -> (then, CastFailed, AsAthena Character Item Definition:Athena Character Item Definition Object Reference) {is=title K2Node_DynamicCast}
   > Set Control Rig(Control Rig:Control Rig Blueprint Object Reference=/EpicControlRig/Templates/MR_Biped_v02) -> (Output_Get:Control Rig Blueprint Object Reference)
   > n2 = Execute Python Script!(
    execute
    PythonScript="import unreal\r\ngenders = {'male':'M', 'female':'F'}\r\nbody_types = {'small':'S', 'medium':'M', 'large':'L'}\r\nfinal_name = ''\r\n\r\ndependecy = unreal.AssetRegistryDependencyOptions(include_hard_management_references =True, include_hard_package_references =True,include_searchable_names=True,include_soft_management_references =True,include_soft_package_references=True)\r\npackage_name = unreal.EditorAssetLibrary.find_asset_data(unreal.EditorAssetLibrary.get_path_name_for_loaded_asset(asset)).package_name\r\n\r\n## Try to get the body type and gender from associated body data file\r\nitem = unreal.load_asset(package_name)\r\nif len(item.base_character_parts) >= 2:\r\n    body = item.base_character_parts[1]\r\n    body_type = body.body_types_permitted.name.lower()\r\n    body_gender = body.gender_permitted.name.lower()\r\n    \r\n    final_name = f'C{body_types.get(body_type,\"\")}{genders.get(body_gender, \"\")}'\r\n\r\nelse:\r\n    ## Otherwise, try to get the body type montage file\r\n    finale_name = ''\r\n    for item in unreal.AssetRegistryHelpers.get_asset_registry().get_dependencies(package_name, dependecy ) or []:\r\n        loaded_item = unreal.load_asset(item)\r\n        if isinstance(loaded_item, unreal.FortMontageLookupTable):\r\n            if unreal.load_asset(item) in list(checktype.keys()):\r\n                final_name = checktype.get(loaded_item, '')\r\n                break"
    asset:Athena Character Item Definition Object Reference=@n1
    checktype:Map of Fort Montage Lookup Table Object References to Strings=Get HumanBodyType Map() -> (HumanBodyType Map:Map of Fort Montage Lookup Table Object References to Strings)
   ) -> (then, ReturnValue, final_name:Text) {inputs [asset,checktype], outputs [final_name]}
   > n3 = Set Text(Text=@n2.final_name, self:Text Box Object Reference=Get SuffixTextBox() -> (SuffixTextBox:Editor Utility Editable Text Box Object Reference))
   n4 = ToLower!(SourceString:String="To String (Text)"(InText=@n3)) -> (ReturnValue:String)
   > n5 = Branch(Condition=OR Boolean(A="Equal Exactly (String)"(A=@n4, B=clf), B="Equal Exactly (String)"(A=@n4, B=clm)))
   > Set IK Retargeter(IK Retargeter:IKRetargeter Object Reference=&/RTG_Emote_Large) -> (Output_Get:IKRetargeter Object Reference)
   n6 = Branch(execute=@n5.else, Condition="Equal Exactly (String)"(A=@n4, B=cmf))
   > Set IK Retargeter(IK Retargeter=&/RTG_Emote_CMF)
   Set IK Retargeter(execute=@n6.else, IK Retargeter=&/RTG_Emote)
   n7 = Execute Python Script!(
    execute=@n1.CastFailed
    PythonScript="import unreal\r\n\r\noutput_skeletal_mesh = None\r\n\r\ndependecy = unreal.AssetRegistryDependencyOptions(include_hard_management_references=True,\r\n                                                  include_hard_package_references=True, include_searchable_names=True,\r\n                                                  include_soft_management_references=True,\r\n                                                  include_soft_package_references=True)\r\npackage_name = unreal.EditorAssetLibrary.find_asset_data(\r\n    unreal.EditorAssetLibrary.get_path_name_for_loaded_asset(asset)).package_name\r\nitem = unreal.load_asset(package_name)\r\n\r\nif isinstance(item, unreal.CosmeticCompanionItemDefinition):\r\n   if len(item.get_character_parts()):\r\n       output_skeletal_mesh = item.get_character_parts()[0].get_skeletal_mesh()\r\n       \r\n"
    asset:Fort Account Item Definition Object Reference=Get CID()
   ) -> (then, ReturnValue, output_skeletal_mesh:Skeletal Mesh Object Reference) {inputs [asset], outputs [output_skeletal_mesh]}
   > Set Companion Skeletal Mesh(Companion Skeletal Mesh:Skeletal Mesh Object Reference=@n7.output_skeletal_mesh) -> (Output_Get:Skeletal Mesh Object Reference)
   > n8 = Is Valid(InputObject=Get Companion Skeletal Mesh() -> (Companion Skeletal Mesh:Skeletal Mesh Object Reference))
   Set Text(execute=@n8.Is Not Valid, self=Get SuffixTextBox())
   n9 = "To String (Text)"(InText=Get Text(self:Text Object Reference=Get SectionNameText_1() -> (SectionNameText_1:Text Object Reference)))
   Set Text(
    execute=@n8.Is Valid
    Text="To Text (String)"(InString=Split(SourceString=@n9, InStr=" ", SearchDir=FromEnd).RightS)
    self=Get SuffixTextBox()
   )
   n10 = Get CID()
   > n11 = Execute Python Script!(
    execute
    PythonScript="import unreal\r\n\r\nsearch_asset = f'MR_{search_asset}'\r\nfound_asset = None\r\n\r\nasset_registry = unreal.AssetRegistryHelpers.get_asset_registry()\r\nall_assets_data = asset_registry.get_assets_by_path(\r\n    \"/CompanionsAnimInEd/\",\r\n    recursive=True\r\n)\r\n\r\nfor asset_data in all_assets_data:\r\n    found_asset = ''\r\n    name = str(asset_data.asset_name)\r\n    if search_asset.lower() in f'MR_{name}'.lower():\r\n        loaded_asset = unreal.AssetData.get_asset(asset_data)\r\n        if isinstance(loaded_asset, unreal.ControlRigBlueprint):\r\n            found_asset = loaded_asset\r\n            break"
    search_asset:String=Split(SourceString=GetObjectName(Object=@n10), InStr=companion_, SearchDir=FromStart).RightS
   ) -> (then, ReturnValue, found_asset:Object Reference) {inputs [search_asset], outputs [found_asset]}
   > n12 = CastToControlRigBlueprint!(execute, Object:Object Reference=@n11.found_asset) -> (CastSucceeded:Exec, CastFailed:Exec, AsControlRigBlueprint:Control Rig Blueprint Object Reference)
   > Set Control Rig(Control Rig=@n12)
   > n13 = Execute Python Script(
    PythonScript="import unreal\r\nfound_asset = None\r\n\r\ndependecy = unreal.AssetRegistryDependencyOptions(include_hard_management_references =True, include_hard_package_references =True,include_searchable_names=True,include_soft_management_references =True,include_soft_package_references=True)\r\npackage_name = unreal.EditorAssetLibrary.find_asset_data(unreal.EditorAssetLibrary.get_path_name_for_loaded_asset(search_asset)).package_name\r\nitem = unreal.load_asset(package_name)\r\n\r\nfor part in item.get_character_parts() or []:\r\n    if isinstance(part, unreal.CustomCharacterPart):\r\n        anim_bp = part.get_anim_class()\r\n        for item in unreal.AssetRegistryHelpers().get_asset_registry().get_dependencies(unreal.StringLibrary.conv_string_to_name(anim_bp.get_package().get_name()), dependecy):\r\n            loaded = unreal.load_asset(item)\r\n            if isinstance(loaded, unreal.IKRetargeter):\r\n                found_asset = loaded\r\n                break"
    search_asset:Fort Account Item Definition Object Reference=Get CID()
   ) {inputs [search_asset], outputs [found_asset]}
   > Is Valid(InputObject=@n13.found_asset)
   > n14 = Cast To IKRetargeter!(execute, Object=@n13.found_asset) -> (then, CastFailed, AsIKRetargeter:IKRetargeter Object Reference) {is=title K2Node_DynamicCast}
   > Set IK Retargeter(IK Retargeter=@n14)
   Event PreConstruct()
   > SetObject(
    self=[
     Get CrigView() -> (CrigView:Single Property View Object Reference)
     Get IKRetargetView() -> (IKRetargetView:Single Property View Object Reference)
     Get CIDView() -> (CIDView:Single Property View Object Reference)
    ]
    NewObject=Self-Reference()
   )
   > SetPropertyName!(execute, self:Single Property View Object Reference=Get CIDView(), NewPropertyName:Name=CID) -> (then)
   > SetPropertyName(self=Get IKRetargetView(), NewPropertyName=IK Retargeter)
   > SetPropertyName(self=Get CrigView(), NewPropertyName=Control Rig)
   Event Construct()
   > Bind Event to E Update Suffix(self:EUW Emote Live Retarget Child Widget Object Reference, Delegate=Create Event()) {is=title K2Node_AddDelegate}
   Event Tick()
  ]}
]
variables [
 CID {type=FortAccountItemDefinition, default=null}
 Companion Skeletal Mesh {type=SkeletalMesh, default=null}
 Control Rig {type=ControlRigBlueprint, default=null, class=~}
 E_Update_Suffix {default=()}
 HumanBodyType Map {default {}}
 IK Retargeter {type=IKRetargeter, default=null}
]
widgets {
 info {rootWidgetClass=&CanvasPanel, widgetCount=27}
 widgets [
  CanvasPanel_33 {widgetClassPath=&CanvasPanel, widget=CanvasPanel_33}
  SizeBox_20 {
   widgetClassPath=&SizeBox
   parent=CanvasPanel_33
   widget=SizeBox_20
   slot {
    class=CanvasPanelSlot
    properties {bAutoSize=true, layoutData {alignment=<0 0>, anchors {maximum=<1 0>, minimum=<0 0>}, offsets=<ltrb 0 0 0 102.9601>}}
   }
  }
  VerticalBox_122 {
   widgetClassPath=&VerticalBox
   parent=SizeBox_20
   widget=VerticalBox_122
   slot {class=SizeBoxSlot, properties {verticalAlignment=VAlign_Center}}
  }
  child_border_widget {
   widgetClassPath=&Border
   bIsVariable=true
   parent=VerticalBox_122
   widget=child_border_widget
   properties {background {
     bIsDynamicallyLoaded=false
     drawAs=Image
     imageType=NoImage
     mirroring=NoMirror
     resourceName=None
     resourceObject=null
     tiling=NoTile
     imageSize=<32 32>
     margin=<ltrb 0 0 0 0>
     outlineSettings {
      bUseBrushTransparency=false
      roundingType=HalfHeightRadius
      width=0
      color {colorUseRule=UseColor_Specified, specifiedColor=<rgba 0 0 0 0>}
      cornerRadii=<0 0 0 0>
     }
     tintColor {colorUseRule=UseColor_Specified, specifiedColor=<rgba 0.026 0.026 0.026 1>}
     uVRegion {bIsValid=false, max=<0 0>, min=<0 0>}
    }}
   slot {class=VerticalBoxSlot}
  }
  VerticalBox_67 {widgetClassPath=&VerticalBox, parent=child_border_widget, widget=VerticalBox_67, slot {class=BorderSlot}}
  SectionNameText_1 {
   widgetClassPath=&TextBlock
   bIsVariable=true
   parent=VerticalBox_67
   widget=SectionNameText_1
   properties {
    text="Target "
    visibility=SelfHitTestInvisible
    colorAndOpacity {colorUseRule=UseColor_Specified, specifiedColor=<rgba 0.5255 0.5255 0.5255 1>}
    font {
     bForceMonospaced=false
     bMaterialIsStencil=false
     fontMaterial=null
     fontObject=&Roboto
     letterSpacing=0
     monospacedWidth=1
     size=9
     skewAmount=0
     typefaceFontName=Bold
     outlineSettings {
      bApplyOutlineToDropShadows=false
      bMiteredCorners=false
      bSeparateFillAlpha=false
      outlineMaterial=null
      outlineSize=0
      outlineColor=<rgba 0 0 0 1>
     }
    }
    strikeBrush {
     bIsDynamicallyLoaded=false
     drawAs=Image
     imageType=NoImage
     mirroring=NoMirror
     resourceName=None
     resourceObject=null
     tiling=NoTile
     imageSize=<32 32>
     margin=<ltrb 0 0 0 0>
     outlineSettings {
      bUseBrushTransparency=false
      roundingType=HalfHeightRadius
      width=0
      color {colorUseRule=UseColor_Specified, specifiedColor=<rgba 0 0 0 0>}
      cornerRadii=<0 0 0 0>
     }
     tintColor {colorUseRule=UseColor_Specified, specifiedColor=<rgba 1 1 1 0.0957>}
     uVRegion {bIsValid=false, max=<0 0>, min=<0 0>}
    }
   }
   slot {class=VerticalBoxSlot}
  }
  HorizontalBox_0 {widgetClassPath=&HorizontalBox, parent=VerticalBox_67, widget=HorizontalBox_0, slot {class=VerticalBoxSlot}}
  IncludeTargetCbox {
   widgetClassPath=&EditorUtilityCheckBox
   bIsVariable=true
   parent=HorizontalBox_0
   widget=IncludeTargetCbox
   properties {checkedState=Checked, toolTipText=Update Asset and re-export Linked Anim}
   slot {class=HorizontalBoxSlot}
  }
  Spacer {
   widgetClassPath=&Spacer
   parent=HorizontalBox_0
   widget=Spacer
   slot {class=HorizontalBoxSlot, properties {padding=<ltrb 0 0 5 0>}}
  }
  CIDView {
   widgetClassPath=&SinglePropertyView
   parent=HorizontalBox_0
   widget=CIDView
   slot {class=HorizontalBoxSlot, properties {size {sizeRule=Fill, value=1}}}
  }
  line_3 {
   widgetClassPath=&CommonBorder
   parent=HorizontalBox_0
   widget=line_3
   properties {
    background {
     bIsDynamicallyLoaded=false
     drawAs=Box
     imageType=NoImage
     mirroring=NoMirror
     resourceName=None
     resourceObject=null
     tiling=NoTile
     imageSize=<18 18>
     margin=<ltrb 0.4 0.4 0.4 0.4>
     outlineSettings {
      bUseBrushTransparency=false
      roundingType=HalfHeightRadius
      width=-9.2626
      color {colorUseRule=UseColor_Specified, specifiedColor=<rgba 0 0 0 0>}
      cornerRadii=<0 0 0 0>
     }
     tintColor {colorUseRule=UseColor_Specified, specifiedColor=<rgba 0 0 0 1>}
     uVRegion {bIsValid=false, max=<0 0>, min=<0 0>}
    }
    padding=<ltrb 1 2 1 2>
    renderTransform {angle=0, scale=<0.8 2.1>, shear=<0 0>, translation=<1.3 0>}
   }
   slot {class=HorizontalBoxSlot}
  }
  Spacer_127 {
   widgetClassPath=&Spacer
   parent=HorizontalBox_0
   widget=Spacer_127
   properties {size=<3 1>}
   slot {class=HorizontalBoxSlot, properties {padding=<ltrb 0 0 16 0>}}
  }
  CrigView {
   widgetClassPath=&SinglePropertyView
   parent=HorizontalBox_0
   widget=CrigView
   slot {class=HorizontalBoxSlot, properties {size {sizeRule=Fill, value=1}}}
  }
  line_2 {
   widgetClassPath=&CommonBorder
   parent=HorizontalBox_0
   widget=line_2
   properties {
    background {
     bIsDynamicallyLoaded=false
     drawAs=Box
     imageType=NoImage
     mirroring=NoMirror
     resourceName=None
     resourceObject=null
     tiling=NoTile
     imageSize=<18 18>
     margin=<ltrb 0.4 0.4 0.4 0.4>
     outlineSettings {
      bUseBrushTransparency=false
      roundingType=HalfHeightRadius
      width=-9.2626
      color {colorUseRule=UseColor_Specified, specifiedColor=<rgba 0 0 0 0>}
      cornerRadii=<0 0 0 0>
     }
     tintColor {colorUseRule=UseColor_Specified, specifiedColor=<rgba 0 0 0 1>}
     uVRegion {bIsValid=false, max=<0 0>, min=<0 0>}
    }
    padding=<ltrb 1 2 1 2>
    renderTransform {angle=0, scale=<0.8 2.1>, shear=<0 0>, translation=<-1.28 0>}
   }
   slot {class=HorizontalBoxSlot}
  }
  Spacer_64 {widgetClassPath=&Spacer, parent=HorizontalBox_0, widget=Spacer_64, properties {size=<8 1>}, slot {class=HorizontalBoxSlot}}
  IKRetargetView {
   widgetClassPath=&SinglePropertyView
   parent=HorizontalBox_0
   widget=IKRetargetView
   slot {class=HorizontalBoxSlot, properties {padding=<ltrb 0 0 -14 0>, size {sizeRule=Fill, value=1}}}
  }
  HorizontalBox_40 {widgetClassPath=&HorizontalBox, parent=VerticalBox_67, widget=HorizontalBox_40, slot {class=VerticalBoxSlot}}
  Spacer_211 {
   widgetClassPath=&Spacer
   parent=HorizontalBox_40
   widget=Spacer_211
   properties {size=<1 1.5>}
   slot {class=HorizontalBoxSlot, properties {size {sizeRule=Fill, value=1}}}
  }
  CanvasPanel {
   widgetClassPath=&CanvasPanel
   parent=VerticalBox_67
   widget=CanvasPanel
   slot {class=VerticalBoxSlot, properties {size {sizeRule=Fill, value=1}}}
  }
  HorizontalBox_1 {
   widgetClassPath=&HorizontalBox
   parent=CanvasPanel
   widget=HorizontalBox_1
   slot {
    class=CanvasPanelSlot
    properties {bAutoSize=true, layoutData {alignment=<0 0>, anchors {maximum=<0.3 0>, minimum=<0 0>}, offsets=<ltrb 0 0 0 30>}}
   }
  }
  LinkedAnimText {
   widgetClassPath=&TextBlock
   parent=HorizontalBox_1
   widget=LinkedAnimText
   properties {
    text=Linked Anim Suffix:
    visibility=HitTestInvisible
    colorAndOpacity {colorUseRule=UseColor_Specified, specifiedColor=<rgba 0.5255 0.5255 0.5255 1>}
    font {
     bForceMonospaced=false
     bMaterialIsStencil=false
     fontMaterial=null
     fontObject=&Roboto
     letterSpacing=0
     monospacedWidth=1
     size=8.1
     skewAmount=0
     typefaceFontName=Regular
     outlineSettings {
      bApplyOutlineToDropShadows=false
      bMiteredCorners=false
      bSeparateFillAlpha=false
      outlineMaterial=null
      outlineSize=0
      outlineColor=<rgba 0 0 0 1>
     }
    }
   }
   slot {class=HorizontalBoxSlot, properties {verticalAlignment=VAlign_Center}}
  }
  CanvasPanel_329 {
   widgetClassPath=&CanvasPanel
   parent=HorizontalBox_1
   widget=CanvasPanel_329
   slot {class=HorizontalBoxSlot, properties {horizontalAlignment=HAlign_Left, size {sizeRule=Fill, value=1}}}
  }
  ScaleBox_47 {
   widgetClassPath=&ScaleBox
   parent=CanvasPanel_329
   widget=ScaleBox_47
   properties {stretch=Fill}
   slot {
    class=CanvasPanelSlot
    properties {bAutoSize=true, layoutData {alignment=<0 0>, anchors {maximum=<1 1>, minimum=<0 0>}, offsets=<ltrb 0 0 0 0>}}
   }
  }
  HorizontalBox_2 {
   widgetClassPath=&HorizontalBox
   parent=ScaleBox_47
   widget=HorizontalBox_2
   slot {class=ScaleBoxSlot, properties {horizontalAlignment=HAlign_Fill}}
  }
  SuffixTextBox {
   widgetClassPath=&EditorUtilityEditableTextBox
   bIsVariable=true
   parent=HorizontalBox_2
   widget=SuffixTextBox
   properties {hintText="######"}
   slot {class=HorizontalBoxSlot}
  }
  Spacer_118 {widgetClassPath=&Spacer, parent=VerticalBox_67, widget=Spacer_118, slot {class=VerticalBoxSlot}}
  line {
   widgetClassPath=&CommonBorder
   parent=VerticalBox_67
   widget=line
   properties {
    background {
     bIsDynamicallyLoaded=false
     drawAs=RoundedBox
     imageType=NoImage
     mirroring=NoMirror
     resourceName=None
     resourceObject=null
     tiling=NoTile
     imageSize=<18 18>
     margin=<ltrb 0.4 0.4 0.4 0.4>
     outlineSettings {
      bUseBrushTransparency=false
      roundingType=HalfHeightRadius
      width=1.4037
      color {colorUseRule=UseColor_Specified, specifiedColor=<rgba 0 0 0 0>}
      cornerRadii=<-0.4977 0.3303 0 -0.5778>
     }
     tintColor {colorUseRule=UseColor_Specified, specifiedColor=<rgba 0 0 0 1>}
     uVRegion {bIsValid=false, max=<0 0>, min=<0 0>}
    }
    renderTransform {angle=0, scale=<1.05 1>, shear=<0 0>, translation=<0 3>}
   }
   slot {class=VerticalBoxSlot}
  }
 ]
}
