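# Editor utility widget for authoring emotes in Sequencer with live retargeting from a source character or companion to multiple targets through BP_LiveRetarget.
# Child widgets define each target's character, Control Rig and IK Retargeter, feeding a process that works on the current Level Sequence.
# It also replaces selected actors with BP_LiveRetarget and opens Animation Browser dialogs for new animations.
# Embedded Python finds companion assets, binds Control Rigs and exports suffixed AnimSequences per skeletal mesh.
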EditorUtilityWidgetBlueprint /AnimInEdTools/EmoteLiveRetarget/EUW_EmoteLiveRetarget
graphs [
 Add Child Widget {nodes [
   Add Child Widget() {is=title K2Node_FunctionEntry}
   > n1 = Create Widget!(
    execute:Exec
    WorldContextObject:Object Reference
    Class:User Widget Class Reference=&/EUW_EmoteLiveRetarget_Child_Widget.EUW_EmoteLiveRetarget_Child_Widget_C
    OwningPlayer:Player Controller Object Reference
   ) -> (then:Exec, ReturnValue:EUW Emote Live Retarget Child Widget Object Reference)
   > Add(
    TargetArray:Array of EUW Emote Live Retarget Child Widget Object References=Get Target Widgets() -> (Target Widgets:Array of EUW Emote Live Retarget Child Widget Object References)
    NewItem:"EUW Emote Live Retarget Child Widget Object Reference (by ref)"=@n1
   ) {is=title K2Node_CallArrayFunction}
   > AddChild(self=Get TargetWidgetsScrollBox() -> (TargetWidgetsScrollBox:Editor Utility Scroll Box Object Reference), Content=@n1)
   > Return Node!(execute, Output Widget:EUW Emote Live Retarget Child Widget Object Reference=@n1)
  ]}
 Add Widget {nodes [
   Add Widget() {is=title K2Node_FunctionEntry}
   n1 = Get Target Widgets()
   > n2 = Branch(Condition="integer < integer"(A=Length(TargetArray:Array of EUW Emote Live Retarget Child Widget Object References=@n1), B=Get max_targets() -> (max_targets:int=0)))
   > n3 = Add Child Widget!(execute, self) -> (then, Output Widget:EUW Emote Live Retarget Child Widget Object Reference) {is=title K2Node_CallFunction}
   n4 = Get Target Widgets()
   > Set Widget Text and Color!(execute, self, Widget:EUW Emote Live Retarget Child Widget Object Reference=@n3, index:int=int - int(A=Length(TargetArray=@n4))) -> (then) {is=title K2Node_CallFunction}
   > SetIsEnabled(self=Get SubWidgetbutton() -> (SubWidgetbutton:Editor Utility Button Object Reference), bInIsEnabled=true)
   SetIsEnabled(execute=@n2.else, self=Get AddWidgetbutton() -> (AddWidgetbutton:Editor Utility Button Object Reference), bInIsEnabled=false)
  ]}
 Call EUW Browser {nodes [
   Call EUW Browser() {is=title K2Node_FunctionEntry}
   > n1 = Create Widget(Class=/AnimInEdTools/AnimationBrowser/EUW_Animation_Dialog.EUW_Animation_Dialog_C) -> (ReturnValue:EUW Animation Dialog Object Reference)
   > Set EUW Anim Dialog(EUW Anim Dialog:EUW Animation Dialog Object Reference=@n1) -> (Output_Get:EUW Animation Dialog Object Reference)
   > AddChild(
    self=Get ModalLayer!(self:Animation Browser Object Reference=Get Anim Browser() -> (Anim Browser:Animation Browser Object Reference)) -> (ModalLayer:Named Slot Object Reference)
    Content=Get EUW Anim Dialog() -> (EUW Anim Dialog:EUW Animation Dialog Object Reference)
   )
   > Bind Event to On Clicked(
    self:Button Object Reference=Get CreateButton!(self:EUW Animation Dialog Object Reference=Get EUW Anim Dialog()) -> (CreateButton:Button Object Reference)
    Delegate=Create Event()
   ) {is=title K2Node_AddDelegate}
   n2 = Get Anim Browser()
   n3 = AnimatableCreature Find!(
    self:Animation Browser Object Reference=@n2
    AnimatableCreatures:Array of Animatable Creature Object References=Get AnimatableCreature!(self:Animation Browser Object Reference=@n2) -> (AnimatableCreature:Array of Animatable Creature Object References)
    Asset:String=GetSelectedOption!(self:"Searchable Combo Box (String) Object Reference"=Get Asset_ComboBox!(self:Animation Browser Object Reference=@n2) -> (Asset_ComboBox:"Searchable Combo Box (String) Object Reference")) -> (ReturnValue:String)
    Variant:String=GetSelectedOption(self=Get Plugin_Combobox!(self:Animation Browser Object Reference=@n2) -> (Plugin_Combobox:"Searchable Combo Box (String) Object Reference"))
   ) -> (AnimatableCreature:Animatable Creature Object Reference, Exact Match:bool)
   > Set Initial Settings!(
    execute
    self:EUW Animation Dialog Object Reference=Get EUW Anim Dialog()
    Project:BP Animation Browser Project Object Reference=Get Current Project!(self:Animation Browser Object Reference=@n2) -> (Current Project:BP Animation Browser Project Object Reference)
    Selected Creature:Animatable Creature Object Reference=@n3
    Developer Mode:bool=Get Developer Mode!(self:Animation Browser Object Reference=@n2) -> (Developer Mode:bool)
    Set Plugin:bool=@n3.Exact Match
    Animatable Creatures:Array of Animatable Creature Object References=Get Animatable Creatures!(
     Project:BP Animation Browser Project Object Reference=Get Current Project(self=@n2)
     bSearchSubClasses:bool=false
     __WorldContext:Object Reference
    ) -> (Animatable Creatures:Array of Animatable Creature Object References) {is=title K2Node_CallFunction}
    Show Project Dialog:bool=false
    Override Anim Path Validation:bool=false
   ) -> (then) {is=title K2Node_CallFunction}
   > Set Title and Button Text!(
    execute
    self:EUW Animation Dialog Object Reference=Get EUW Anim Dialog()
    Title:String=Create New Emote Animation
    Create Button:String=Create
    Description:String=Create a new Level Sequence and Linked AnimSequences for master and set targets
   ) -> (then) {is=title K2Node_CallFunction}
   n4 = GetSelectedOption(self=Get Subfolder_Combobox!(self:Animation Browser Object Reference=Get Anim Browser()) -> (Subfolder_Combobox:"Searchable Combo Box (String) Object Reference"))
   > Branch(Condition="Not Equal, Case Insenstive (String)"!(A:String=@n4, B:String=All) -> (ReturnValue:bool))
   > Set Subfolder!(execute, self:EUW Animation Dialog Object Reference=Get EUW Anim Dialog(), Option:String=@n4) -> (then) {is=title K2Node_CallFunction}
  ]}
 Conditional Add Widget {nodes [
   Conditional Add Widget!() -> (
    then
    Athena Item Array:Array of Athena Character Item Definition Object References
    Companions:bool=~
    Companions Array:Array of Skeletal Mesh Object References
   ) {is=title K2Node_FunctionEntry}
   > n1 = Sequence!(execute) -> (then_0, then_1, then_2:Exec)
   > n2 = Branch(Condition=Get Companions() -> (Companions:bool))
   > n3 = For Each Loop(Array:Array of Skeletal Mesh Object References=Get Companions Array() -> (Companions Array:Array of Skeletal Mesh Object References)) -> (Array Element:Skeletal Mesh Object Reference)
   > Is Valid(InputObject=@n3)
   > Set Local Array Length(Local Array Length:int=int + int(A=Get Local Array Length() -> (Local Array Length:int=0))) -> (Output_Get:int=0)
   n4 = For Each Loop(
    Exec=@n2.else
    Array:Array of Athena Character Item Definition Object References=Get Athena Item Array() -> (Athena Item Array:Array of Athena Character Item Definition Object References)
   ) -> (Array Element:Athena Character Item Definition Object Reference)
   > Is Valid(InputObject=@n4)
   > Set Local Array Length(Local Array Length=int + int(A=Get Local Array Length() -> (Local Array Length=0))) -> (Output_Get=0)
   n5 = Length(TargetArray=Get Target Widgets())
   Branch(execute=@n1.then_1, Condition="integer <= integer"(A=int + int(A=@n5), B=Get Local Array Length() -> (Local Array Length=0)))
   > Add Widget() {is=title K2Node_CallFunction}
   Return Node!(execute=@n1.then_2)
  ]}
 Convert to Spawnable {nodes [
   Convert to Spawnable!() -> (then, Input Level Sequence:Level Sequence Object Reference, ActorClass:Actor Class Reference) {is=title K2Node_FunctionEntry}
   > n1 = Sequence()
   > n2 = GetBindings!(execute, Sequence:Movie Scene Sequence Object Reference=Get Input Level Sequence() -> (Input Level Sequence:Level Sequence Object Reference)) -> (then, ReturnValue:Array of Movie Scene Binding Proxy Structures)
   > n3 = For Each Loop(Array:Array of Movie Scene Binding Proxy Structures=@n2) -> (Array Element:Movie Scene Binding Proxy Structure)
   > n4 = Branch(Condition=Get Cast Status() -> (Cast Status:bool))
   Set Local Binding(execute=@n4.else, Local Binding:Movie Scene Binding Proxy Structure=@n3) -> (Output_Get:Movie Scene Binding Proxy Structure)
   > n5 = GetBindingID!(
    execute
    Sequence:Movie Scene Sequence Object Reference=Get Input Level Sequence()
    InBinding:"Movie Scene Binding Proxy Structure (by ref)"=Get Local Binding() -> (Local Binding:Movie Scene Binding Proxy Structure)
   ) -> (then, ReturnValue:Movie Scene Object Binding ID Structure)
   > n6 = For Each Loop(Array:Array of Object References=GetBoundObjects!(ObjectBinding:Movie Scene Object Binding ID Structure=@n5) -> (ReturnValue:Array of Object References)) -> (Array Element:Object Reference)
   > n7 = Branch(Condition=Get Cast Status())
   Cast To BP_LiveRetarget!(execute=@n7.else, Object=@n6) -> (then, CastFailed, AsBP Live Retarget:BP Live Retarget Object Reference) {is=title K2Node_DynamicCast}
   > Set Cast Status(Cast Status:bool=true) -> (Output_Get:bool)
   n8 = Branch(execute=@n1.then_1, Condition=Get Cast Status())
   n9 = Sequence(execute=@n8.else)
   > n10 = SpawnActorFromClass!(
    execute
    ActorClass:Actor Class Reference=Get ActorClass() -> (ActorClass:Actor Class Reference)
    Location:Vector="0, 0, 0"
    Rotation:Rotator
    bTransient:bool
   ) -> (then, ReturnValue:Actor Object Reference)
   > n11 = AddPossessable!(execute, Sequence:Movie Scene Sequence Object Reference=Get Input Level Sequence(), ObjectToPossess:Object Reference=@n10) -> (then, ReturnValue:Movie Scene Binding Proxy Structure)
   > n12 = ConvertToSpawnable!(
    execute
    self:Level Sequence Editor Subsystem Object Reference=Get LevelSequenceEditorSubsystem() -> (ReturnValue:Level Sequence Editor Subsystem Object Reference) {is=title K2Node_GetEditorSubsystem}
    ObjectBinding:"Movie Scene Binding Proxy Structure (by ref)"=@n11
   ) -> (then, ReturnValue:Array of Movie Scene Binding Proxy Structures)
   > n13 = GetBindingID(
    Sequence=Get Input Level Sequence()
    InBinding="Get (a copy)"(Array:Array of Movie Scene Binding Proxy Structures=@n12) -> (Output:Movie Scene Binding Proxy Structure)
   )
   > n14 = ResolveBindingID!(
    execute
    RootSequence:Movie Scene Sequence Object Reference=Get Input Level Sequence()
    InObjectBindingID:Movie Scene Object Binding ID Structure=@n13
   ) -> (then, ReturnValue:Movie Scene Binding Proxy Structure)
   > Set Local Binding(Local Binding=@n14)
   Destroy Actor(execute=@n9.then_1, self=@n10)
   Return Node!(execute=@n1.then_2, "Output Prorxy ":Movie Scene Binding Proxy Structure=Get Local Binding())
  ]}
 Create New Linked Anim Per Component {nodes [
   Create New Linked Anim Per Component!() -> (then, Input Target Widgets:Array of EUW Emote Live Retarget Child Widget Object References) {is=title K2Node_FunctionEntry}
   > n1 = GetCurrentLevelSequence!(execute) -> (then, ReturnValue:Level Sequence Object Reference)
   > Set Local Current Level Sequence(Local Current Level Sequence:Level Sequence Object Reference=@n1) -> (Output_Get:Level Sequence Object Reference)
   > n2 = Sequence!(execute) -> (then_0, then_1)
   > n3 = Get Animation!(
    execute
    self
    Input Sequence:Movie Scene Sequence Object Reference=Get Local Current Level Sequence() -> (Local Current Level Sequence:Level Sequence Object Reference)
   ) -> (then, Ouput Params Animation:Anim Sequence Base Object Reference, Output Animation Section:Movie Scene Section Object Reference) {is=title K2Node_CallFunction}
   > Set Local Params Animation(Local Params Animation:Anim Sequence Base Object Reference=@n3) -> (Output_Get:Anim Sequence Base Object Reference)
   n4 = GetBindings(execute=@n2.then_1, Sequence=Get Local Current Level Sequence())
   > n5 = For Each Loop(Array:Array of Movie Scene Binding Proxy Structures=@n4) -> (Array Element:Movie Scene Binding Proxy Structure)
   > Set Local Binding(Local Binding=@n5)
   > n6 = Get Linked Animation Path!(
    execute
    self
    InLevelSequence:Level Sequence Object Reference=Get Local Current Level Sequence()
    MovieSceneBindingProxy:"Movie Scene Binding Proxy Structure (by ref)"=Get Local Binding()
   ) -> (then, Output Animation Path:String) {is=title K2Node_CallFunction}
   > Set Local Current Linked Anim Path(Local Current Linked Anim Path:String=@n6) -> (Output_Get:String)
   > n7 = GetBindingID(Sequence=Get Local Current Level Sequence(), InBinding=Get Local Binding())
   > n8 = For Each Loop(Array:Array of Object References=GetBoundObjects(ObjectBinding=@n7)) -> (Array Element:Object Reference)
   > n9 = Cast To SkeletalMeshComponent!(execute, Object=@n8) -> (then, CastFailed, AsSkeletal Mesh Component:Skeletal Mesh Component Object Reference) {is=title K2Node_DynamicCast}
   n10 = Get SkeletalMesh!(self:Skinned Mesh Component Object Reference=@n9) -> (SkeletalMesh:Skeletal Mesh Object Reference)
   > n11 = Is Valid(InputObject=@n10)
   PrintString(execute=@n11.Is Not Valid, InString=Append(A=Skipping, B=GetDisplayName(Object=@n8)))
   Set Local Skeletal Mesh(execute=@n11.Is Valid, Local Skeletal Mesh:Skeletal Mesh Object Reference=@n10) -> (Output_Get:Skeletal Mesh Object Reference)
   > n12 = Cast To BP_LiveRetarget(Object=GetOwner(self=@n9))
   n13 = Find Item!(
    TargetArray:Array of Skeletal Mesh Component Object References=Get Target Skeletal Meshes!(self:BP Live Retarget Object Reference=@n12) -> (Target Skeletal Meshes:Array of Skeletal Mesh Component Object References)
    ItemToFind:"Skeletal Mesh Component Object Reference (by ref)"=@n9
   ) -> (ReturnValue)
   > n14 = Branch(Condition="Not Equal (Integer)"(A=@n13, B=-1))
   n15 = "Get (a copy)"(
    Array:Array of EUW Emote Live Retarget Child Widget Object References=Get Input Target Widgets() -> (Input Target Widgets:Array of EUW Emote Live Retarget Child Widget Object References)
    Dimension 1=@n13
   ) -> (Output:EUW Emote Live Retarget Child Widget Object Reference)
   > Branch(Condition=IsChecked(self=Get IncludeTargetCbox!(self:EUW Emote Live Retarget Child Widget Object Reference=@n15) -> (IncludeTargetCbox:Editor Utility Check Box Object Reference)))
   > n16 = Execute Python Script!(
    execute
    PythonScript="import unreal\r\n\r\nanim_suffix = ''\r\n\r\ngenders = {'male':'M', 'female':'F'}\r\nbody_types = {'small':'S', 'medium':'M', 'large':'L'}\r\n\r\n## first index passed in is always the master (skeletal component)\r\nif index <0:\r\n\tanim_suffix = 'master'\r\n## Check if the index - 1 is set via the EUW \r\nelse:\r\n\tif len(suffix_array) > index:\r\n\t\tcurrent_widget = suffix_array[index]\r\n\t\ttextbox_widget = current_widget.find_child_widget_by_name('SuffixTextBox')\r\n\t\tif textbox_widget:\r\n\t\t\tanim_suffix = f'{str(textbox_widget.get_text())}'\r\n\r\n\t\t## if not, use the same approach as the EUW to directly on the CID object in the current sequence BP\r\n\t\tif not anim_suffix:\r\n\t\t\tasset = target_cids[index]\r\n\t\t\tprint(asset)\r\n\t\t\tdependecy = unreal.AssetRegistryDependencyOptions(include_hard_management_references =True, include_hard_package_references =True,include_searchable_names=True,include_soft_management_references =True,include_soft_package_references=True)\r\n\t\t\tpackage_name = unreal.EditorAssetLibrary.find_asset_data(unreal.EditorAssetLibrary.get_path_name_for_loaded_asset(asset)).package_name\r\n\r\n\t\t\t## Try to get the body type and gender from associated body data file\r\n\t\t\titem = unreal.load_asset(package_name)\r\n\t\t\tif len(item.base_character_parts) >= 2:\r\n\t\t\t\tbody = item.base_character_parts[1]\r\n\t\t\t\tbody_type = body.body_types_permitted.name.lower()\r\n\t\t\t\tbody_gender = body.gender_permitted.name.lower()\r\n\t\t\t\tanim_suffix = f'C{body_types.get(body_type,\"\")}{genders.get(body_gender,\"\")}'\r\n\r\n\t\t\telse:\r\n\t\t\t\t## Otherwise, try to get the body type montage file\r\n\t\t\t\tfor item in unreal.AssetRegistryHelpers.get_asset_registry().get_dependencies(package_name, dependecy ) or []:\r\n\t\t\t\t\tloaded_item = unreal.load_asset(item)\r\n\t\t\t\t\tif isinstance(loaded_item, unreal.FortMontageLookupTable):\r\n\t\t\t\t\t\tif unreal.load_asset(item) in list(checktype.keys()):\r\n\t\t\t\t\t\t\tanim_suffix = f'{str(checktype.get(loaded_item, \"\"))}'\r\n\t\t\t\t\t\t\tbreak\r\n\r\nif anim_suffix:\r\n\tanim_suffix = f'_{anim_suffix}'"
    suffix_array:Array of EUW Emote Live Retarget Child Widget Object References=Get Input Target Widgets()
    index:int=@n13
    target_cids:Array of Athena Character Item Definition Object References=Get Target CIDs!(self:BP Live Retarget Object Reference=@n12) -> (Target CIDs:Array of Athena Character Item Definition Object References)
   ) -> (then, ReturnValue, anim_suffix:String) {inputs [suffix_array,index,target_cids], outputs [anim_suffix]}
   > n17 = Set Local Anim Suffix(Local Anim Suffix:String=@n16.anim_suffix) -> (Output_Get:String)
   n18 = Branch(execute=[@n17.then, @n29.then], Condition=IsEmpty(InString=Get Anim Sequence() -> (Anim Sequence:String)))
   > PrintString(InString=Checking for Linked Sequence...)
   > n19 = Branch(Condition=IsEmpty(InString=Get Local Current Linked Anim Path() -> (Local Current Linked Anim Path:String)))
   PrintString(execute=[@n25.CastFailed, @n19.then], InString=No linked anim...)
   n20 = Contains(SearchIn=Get Local Anim Suffix() -> (Local Anim Suffix:String), Substring=_master, bSearchFromEnd=true)
   > Branch(Condition=AND Boolean(A=Get replace() -> (replace:bool), B=InvertBool!(bValue:bool=@n20) -> (ReturnValue:bool)))
   > n21 = GetParent!(execute, InBinding:"Movie Scene Binding Proxy Structure (by ref)"=Get Local Binding()) -> (then, ReturnValue:Movie Scene Binding Proxy Structure)
   > n22 = Get Linked Animation Path(InLevelSequence=Get Local Current Level Sequence(), MovieSceneBindingProxy=@n21)
   > n23 = Branch(Condition=IsEmpty(InString=@n22))
   Execute Python Script!(
    execute=@n23.else
    PythonScript="import unreal\r\n##Grab the Level Editor World\r\neditor_subsystem = unreal.get_editor_subsystem(unreal.UnrealEditorSubsystem)\r\nworld = editor_subsystem.get_editor_world()\r\n\r\n##Create animation sequence export options\r\nanim_seq_export_options = unreal.AnimSeqExportOption()\r\nanim_seq_export_options.export_transforms = True\r\nanim_seq_export_options.export_morph_targets = True\r\nanim_seq_export_options.set_retarget_source_asset = True\r\n\r\n##Asset Tools\r\nassets_tools = unreal.AssetToolsHelpers.get_asset_tools()\r\n\r\n##create anim sequence factory and assign target skeleton\r\nanim_sequence_factory = unreal.AnimSequenceFactory()\r\nanim_sequence_factory.target_skeleton = skeleton\r\nanim_sequence = unreal.AssetTools.create_asset(assets_tools, asset_name=asset_name, package_path=package_path, asset_class=unreal.AnimSequence, factory=anim_sequence_factory)\r\n\r\n##Export\r\nunreal.ControlRigSequencerLibrary.export_anim_sequence_from_sequencer(anim_sequence, anim_seq_export_options, binding, True)"
    package_path:String=GetPath!(InPath:String=@n22) -> (ReturnValue:String)
    asset_name:String=Append(A=GetBaseFilename(InPath=@n22, bRemovePath=true), B=Get Local Anim Suffix())
    level_sequence:Level Sequence Object Reference=Get Local Current Level Sequence()
    binding:Movie Scene Binding Proxy Structure=Get Local Binding()
    skeleton:Skeleton Object Reference=Get Skeleton!(self:Skeletal Mesh Object Reference=Get Local Skeletal Mesh() -> (Local Skeletal Mesh:Skeletal Mesh Object Reference)) -> (Skeleton:Skeleton Object Reference)
   ) -> (then, ReturnValue) {inputs [package_path, asset_name, level_sequence, binding, skeleton]}
   n24 = LoadAsset(execute=[@n19.else, @n26.else], AssetPath=Get Local Current Linked Anim Path())
   > n25 = Cast To AnimSequence!(execute, Object=@n24) -> (then, CastFailed, AsAnim Sequence:Anim Sequence Object Reference)
   > Execute Python Script!(
    execute
    PythonScript="import unreal\r\n\r\ncomp = component.get_name()\r\nprint(comp)\r\nif 'MALE' in comp:\r\n    comp = f'MPR_{comp}'\r\nprint(comp)\r\n\r\nanim_sequence.set_editor_property('retarget_source', unreal.Name(comp))"
    anim_sequence:Anim Sequence Object Reference=@n25
    component:Skeletal Mesh Object Reference=Get Local Skeletal Mesh()
   ) -> (then, ReturnValue) {inputs [anim_sequence, component]}
   > CheckoutAsset!(execute, AssetToCheckout:String=Get Local Current Linked Anim Path()) -> (then, ReturnValue:bool)
   > SaveAsset!(execute, AssetToSave:String=Get Local Current Linked Anim Path(), bOnlyIfIsDirty:bool=true) -> (then, ReturnValue:bool)
   n26 = Branch(execute=@n18.else, Condition=IsEmpty(InString=Get Local Current Linked Anim Path()))
   n27 = Append(A=Split(SourceString=Get Anim Sequence(), InStr=/, SearchDir=FromEnd).RightS, B=Get Local Anim Suffix())
   n28 = Get Animation Path() -> (Animation Path:String)
   > CheckoutAsset(AssetToCheckout=GetPath(InPath=Append!(A=@n28, B=@n27, C:String) -> (ReturnValue)))
   > Execute Python Script!(
    execute
    PythonScript="import unreal\r\n##Grab the Level Editor World\r\neditor_subsystem = unreal.get_editor_subsystem(unreal.UnrealEditorSubsystem)\r\nworld = editor_subsystem.get_editor_world()\r\n\r\n##Create animation sequence export options\r\nanim_seq_export_options = unreal.AnimSeqExportOption()\r\nanim_seq_export_options.export_transforms = True\r\nanim_seq_export_options.export_morph_targets = True\r\nanim_seq_export_options.set_retarget_source_asset = True\r\n\r\n##Asset Tools\r\nassets_tools = unreal.AssetToolsHelpers.get_asset_tools()\r\n\r\n##create anim sequence factory and assign target skeleton\r\nanim_sequence_factory = unreal.AnimSequenceFactory()\r\nanim_sequence_factory.target_skeleton = skeleton\r\nanim_sequence = unreal.AssetTools.create_asset(assets_tools, asset_name=asset_name, package_path=package_path, asset_class=unreal.AnimSequence, factory=anim_sequence_factory)\r\n\r\n##Export\r\nunreal.ControlRigSequencerLibrary.export_anim_sequence_from_sequencer(anim_sequence, anim_seq_export_options, binding, True)"
    package_path:String=@n28
    asset_name:String=@n27
    level_sequence:Level Sequence Object Reference=Get Local Current Level Sequence()
    binding:Movie Scene Binding Proxy Structure=Get Local Binding()
    skeleton:Skeleton Object Reference=Get Skeleton(self=Get Local Skeletal Mesh())
   ) -> (then, ReturnValue) {inputs [package_path, asset_name, level_sequence, binding, skeleton]}
   Branch(execute=@n14.else, Condition=IsChecked(self=Get IncludeSourcecbox() -> (IncludeSourcecbox:Editor Utility Check Box Object Reference)))
   > n29 = Set Local Anim Suffix(Local Anim Suffix=_master)
  ]}
 E Bind Control Rig {nodes [E Bind Control Rig() {is=title K2Node_FunctionEntry}]}
 E Create New Animation {nodes [E Create New Animation() {is=title K2Node_FunctionEntry}]}
 E Export Animations {nodes [E Export Animations() {is=title K2Node_FunctionEntry}]}
 E Main Process {nodes [E Main Process() {is=title K2Node_FunctionEntry}]}
 E Remove Widget {nodes [E Remove Widget() {is=title K2Node_FunctionEntry}]}
 EventGraph {nodes [
   "On Clicked (CreateNewSequenceButton)"() {is=title K2Node_ComponentBoundEvent}
   > Set replace(replace:bool=false) -> (Output_Get:bool)
   > Set Anim Sequence(Anim Sequence:String) -> (Output_Get:String)
   > Set Animation Path(Animation Path:String) -> (Output_Get:String)
   > n1 = Branch(Condition=IsChecked(self=Get GenerateAnimswhenAddCbox() -> (GenerateAnimswhenAddCbox:Editor Utility Check Box Object Reference)))
   > n2 = GetCurrentLevelSequence()
   > Is Valid(InputObject=@n2)
   > n3 = GetBindings(Sequence=@n2)
   > n4 = For Each Loop with Break!(Exec:Exec, Array:Array of Movie Scene Binding Proxy Structures=@n3, Break:Exec=@n7.else) -> (LoopBody:Exec, Array Element:Movie Scene Binding Proxy Structure, Array Index:int, Completed:Exec)
   > n5 = Get Linked Animation Path(InLevelSequence=@n2, MovieSceneBindingProxy=@n4)
   n6 = IsEmpty(InString=@n5)
   > n7 = Branch(Condition=@n6)
   n8 = Branch(execute=@n4.Completed, Condition=@n6)
   > n9 = Construct Object from Class(Class=&EUW_Animation_Dialog_C) -> (ReturnValue:EUW Animation Dialog Object Reference)
   > AddChild(self=Get anim_dialog() -> (anim_dialog:Named Slot Object Reference), Content=@n9)
   > Set Initial Settings(
    self=@n9
    Developer Mode=Get Developer Mode(self=Get Anim Browser())
    Set Plugin=false
    Show Project Dialog=true
    Override Anim Path Validation=true
   )
   > Set Title and Button Text(
    self=@n9
    Title="Set Anim Output Path "
    Create Button=Add to Current Sequence
    Description=No Linked Anims found for BP LiveRetarget in this sequence. Please set animation path
   )
   > Bind Event to Generate Animation Paths(self:EUW Animation Dialog Object Reference=@n9, Delegate=Create Event()) {is=title K2Node_AddDelegate}
   Set Anim Sequence(execute=@n8.else, Anim Sequence=Split(SourceString=@n5, InStr=., SearchDir=FromStart))
   > Set Animation Path(Animation Path=Split(SourceString=@n5, InStr=/, SearchDir=FromEnd))
   > add update event()
   add update event(execute=@n1.else)
   "On Clicked (SubWidgetbutton)"() {is=title K2Node_ComponentBoundEvent}
   > Call E Remove Widget!(execute:Exec, self:EUW Emote Live Retarget Object Reference) -> (then:Exec) {is=title K2Node_CallDelegate}
   "On Pressed (ReplaceSelectedActortoEmoteRetargetButton)"() {is=title K2Node_ComponentBoundEvent}
   > n10 = For Each Loop(Array:Array of Movie Scene Binding Proxy Structures=GetSelectedBindings!() -> (ReturnValue:Array of Movie Scene Binding Proxy Structures)) -> (Array Element:Movie Scene Binding Proxy Structure)
   n11 = Get binding() -> (binding:Movie Scene Binding Proxy Structure)
   n12 = GetPossessedObjectClass!(execute=@n10.Completed, InBinding:"Movie Scene Binding Proxy Structure (by ref)"=@n11) -> (then, ReturnValue:Object Class Reference)
   > n13 = Branch(Condition=AND Boolean(
     A="Not Equal (Class)"(A:Object Class Reference=@n12, B:Object Class Reference=&/BP_LiveRetarget.BP_LiveRetarget_C) -> (ReturnValue:bool=~) {is=title K2Node_PromotableOperator}
     B=IsValid!(InBinding:"Movie Scene Binding Proxy Structure (by ref)"=@n11) -> (ReturnValue)
    ))
   > n14 = GetName!(execute, InBinding:"Movie Scene Binding Proxy Structure (by ref)"=Get binding()) -> (then, ReturnValue:String)
   > Set replace(replace=true)
   > n15 = Show Message Dialog(
    Title="To Text (String)"(InString=Replace Actor With BP_LIveRetarget)
    Message="To Text (String)"(InString=Append(
      A="Will Attempt to replace: "
      B=@n14
      C=" with BP Live Retarget with the UI settings! \r\nThis will keep the current rigs attached to the current Binding. \r\nContinue?"
     ))
    MessageType=YesNo
    DefaultValue=No
    MessageCategory=Info
   )
   > n16 = Branch(Condition="Equal (Enum)"(A:EAppReturnType Enum=@n15, B:EAppReturnType Enum=Yes))
   > ChangeActorTemplateClass!(
    execute
    self:Level Sequence Editor Subsystem Object Reference=Get LevelSequenceEditorSubsystem()
    ObjectBinding:"Movie Scene Binding Proxy Structure (by ref)"=Get binding()
    ActorClass:Actor Class Reference=&BP_LiveRetarget_C
   ) -> (then, ReturnValue:bool)
   > Set Anim Sequence()
   > n17 = Set Animation Path()
   n18 = check companion skeletons!(execute=[@n36.then, @n92.then, @n17.then], self) -> (then, Continue:EAppReturnType Enum)
   > Branch(Condition="Equal (Enum)"(A=@n18, B=Yes))
   > Call E Main Process!(execute:Exec, self:EUW Emote Live Retarget Object Reference) -> (then:Exec) {is=title K2Node_CallDelegate}
   Set replace(execute=@n16.else, replace=false)
   PrintString(execute=@n13.else, InString=Nothing selected/invalid selection. Selected item must be a BP that is not already BP_LiveRetarget)
   n19 = GetParent(execute=@n10.LoopBody, InBinding=@n10)
   > n20 = Branch(Condition=IsValid(InBinding=@n19))
   > Set binding(binding:Movie Scene Binding Proxy Structure=@n19) -> (Output_Get:Movie Scene Binding Proxy Structure)
   Set binding(execute=@n20.else, binding=@n10)
   n21 = "On Clicked (ClearButton)"() {is=title K2Node_ComponentBoundEvent}
   n22 = Sequence!(execute=[@n21.then, @n90.then]) -> (then_0, then_1, then_2:Exec)
   > Set Parent CID(Parent CID:Fort Account Item Definition Object Reference) -> (Output_Get:Fort Account Item Definition Object Reference)
   > Set Parent Control Rig(Parent Control Rig:Control Rig Blueprint Object Reference) -> (Output_Get:Control Rig Blueprint Object Reference)
   > Set Source Animation(Source Animation:Anim Sequence Object Reference) -> (Output_Get:Anim Sequence Object Reference)
   n23 = "Get (a copy)"(Array=Get Target Widgets())
   Set CID!(execute=@n22.then_1, CID:Fort Account Item Definition Object Reference, self:EUW Emote Live Retarget Child Widget Object Reference=@n23) -> (then, Output_Get:Fort Account Item Definition Object Reference)
   > Set Control Rig!(execute, Control Rig:Control Rig Blueprint Object Reference, self:EUW Emote Live Retarget Child Widget Object Reference=@n23) -> (then, Output_Get:Control Rig Blueprint Object Reference)
   > Set IK Retargeter!(execute, IK Retargeter:IKRetargeter Object Reference, self:EUW Emote Live Retarget Child Widget Object Reference=@n23) -> (then, Output_Get:IKRetargeter Object Reference)
   n24 = Set bIsEnabled!(execute=@n22.then_2, bIsEnabled:bool=true, self:Widget Object Reference=Get SubWidgetbutton()) -> (then, Output_Get:bool)
   > While Loop!(execute:Exec, Condition:bool=@n24) -> (LoopBody:Exec, Completed:Exec) {is=title K2Node_MacroInstance}
   > Call E Remove Widget()
   "On Clicked (UpdateDetailsButton)"() {is=title K2Node_ComponentBoundEvent}
   > clear all()
   > n25 = GetCurrentLevelSequence()
   > Update Widges from Sequence!(execute, self, Input Level Sequence:Level Sequence Object Reference=@n25) -> (then)
   "On Clicked (?button)"() {is=title K2Node_ComponentBoundEvent}
   > Execute Python Script!(
    execute
    PythonScript="import webbrowser\r\n\r\nwebbrowser.open('https://epicgames.atlassian.net/wiki/spaces/FORT/pages/484182538/Emote+Live+Retarget+Tool')"
   ) -> (then, ReturnValue)
   n26 = "On Property Changed (ParentCIDView)"!() -> (OutputDelegate, then, PropertyName:Name=None) {is=title K2Node_ComponentBoundEvent}
   n27 = Sequence!(execute=[@n26.then, @n41.then]) -> (then_0, then_1)
   > n28 = Execute Python Script!(
    execute
    PythonScript="import unreal\r\n\r\noutput_skeletal_mesh = None\r\n\r\ndependecy = unreal.AssetRegistryDependencyOptions(include_hard_management_references=True,\r\n                                                  include_hard_package_references=True, include_searchable_names=True,\r\n                                                  include_soft_management_references=True,\r\n                                                  include_soft_package_references=True)\r\npackage_name = unreal.EditorAssetLibrary.find_asset_data(\r\n    unreal.EditorAssetLibrary.get_path_name_for_loaded_asset(asset)).package_name\r\nitem = unreal.load_asset(package_name)\r\n\r\nif isinstance(item, unreal.CosmeticCompanionItemDefinition):\r\n   if len(item.get_character_parts()):\r\n       output_skeletal_mesh = item.get_character_parts()[0].get_skeletal_mesh()\r\n       \r\n"
    asset:Fort Account Item Definition Object Reference=Get Parent CID() -> (Parent CID:Fort Account Item Definition Object Reference)
   ) -> (then, ReturnValue, output_skeletal_mesh:Skeletal Mesh Object Reference) {inputs [asset], outputs [output_skeletal_mesh]}
   > Set Parent Companion Skeletal Mesh(Parent Companion Skeletal Mesh:Skeletal Mesh Object Reference=@n28.output_skeletal_mesh) -> (Output_Get:Skeletal Mesh Object Reference)
   Is Valid(exec=@n27.then_1, InputObject=Get Parent Companion Skeletal Mesh() -> (Parent Companion Skeletal Mesh:Skeletal Mesh Object Reference))
   n29 = Get Parent CID()
   > n30 = Execute Python Script!(
    execute
    PythonScript="import unreal\r\n\r\nsearch_asset = f'MR_{search_asset}'\r\n\r\nasset_registry = unreal.AssetRegistryHelpers.get_asset_registry()\r\nall_assets_data = asset_registry.get_assets_by_path(\r\n    \"/CompanionsAnimInEd/\",\r\n    recursive=True\r\n)\r\n\r\nfor asset_data in all_assets_data:\r\n    found_asset = ''\r\n    name = str(asset_data.asset_name)\r\n    if search_asset.lower() in f'MR_{name}'.lower():\r\n        loaded_asset = unreal.AssetData.get_asset(asset_data)\r\n        if isinstance(loaded_asset, unreal.ControlRigBlueprint):\r\n            found_asset = loaded_asset\r\n            break"
    search_asset:String=Split(SourceString=GetObjectName(Object=@n29), InStr=companion_, SearchDir=FromStart).RightS
   ) -> (then, ReturnValue, found_asset:Object Reference) {inputs [search_asset], outputs [found_asset]}
   > n31 = CastToControlRigBlueprint!(execute, Object:Object Reference=@n30.found_asset) -> (CastSucceeded:Exec, CastFailed:Exec, AsControlRigBlueprint:Control Rig Blueprint Object Reference)
   > Set Parent Control Rig(Parent Control Rig=@n31)
   "On Clicked (AddWidgetbutton)"() {is=title K2Node_ComponentBoundEvent}
   > Add Widget()
   "On Clicked (NewAnimationButton)"() {is=title K2Node_ComponentBoundEvent}
   > SetIsChecked!(execute, self:Check Box Object Reference=Get IncludeSourcecbox(), InIsChecked:bool=true) -> (then)
   > Call EUW Browser() {is=title K2Node_CallFunction}
   Bind Control Rig() {is=title K2Node_CustomEvent}
   > Execute Python Script!(
    execute
    PythonScript="import unreal\r\n\r\neditor_system = unreal.get_editor_subsystem(unreal.UnrealEditorSubsystem)\r\nworld = editor_system.get_editor_world()\r\n\r\n## Get the rig class\r\nrig_class = control_rig.get_control_rig_class()\r\n\r\nunreal.ControlRigSequencerLibrary.find_or_create_control_rig_track(\r\n    world = world,\r\n    level_sequence = level_sequence,\r\n    control_rig_class = rig_class,\r\n    binding  = target_binding,\r\n    is_layered_control_rig = layered\r\n        )"
    control_rig:Control Rig Blueprint Object Reference=Get Current Control Rig() -> (Current Control Rig:Control Rig Blueprint Object Reference)
    level_sequence:Level Sequence Object Reference=Resolve Soft Reference!(Input:Level Sequence Soft Object Reference=Get Current Level Sequence() -> (Current Level Sequence:Level Sequence Soft Object Reference)) -> (Output:Level Sequence Object Reference)
    target_binding:Movie Scene Binding Proxy Structure=Get Target Binding() -> (Target Binding:Movie Scene Binding Proxy Structure)
    layered:bool="Get Layered Rig "() -> ("Layered Rig ":bool)
   ) -> (then, ReturnValue) {inputs [control_rig, level_sequence, target_binding, layered]}
   Generate Animations() {is=title K2Node_CustomEvent}
   > Branch(Condition=IsChecked(self=Get GenerateAnimswhenAddCbox()))
   > Create New Linked Anim Per Component!(execute, self, Input Target Widgets:Array of EUW Emote Live Retarget Child Widget Object References=Get Target Widgets()) -> (then) {is=title K2Node_CallFunction}
   n32 = Animation Changed!() -> (OutputDelegate, then, Item:Object Reference) {is=title K2Node_CustomEvent}
   > n33 = Cast To Object_AnimationBrowser!(execute, Object=@n32.Item) -> (then, CastFailed, AsObject Animation Browser:Object Animation Browser Object Reference) {is=title K2Node_DynamicCast}
   > n34 = LoadAsset_Blocking(Asset=Get!(self:Object Animation Browser Object Reference=@n33) -> (
     Data_AnimSequence_31_#1:Soft Object Reference
     Data_LevelSequence_32_#2:Soft Object Reference
     Data_CurrentUser_26_#3:String
    ))
   > n35 = Cast To AnimSequence(Object=@n34)
   > Set Source Animation(Source Animation=@n35)
   n36 = add update event() {is=title K2Node_CustomEvent}
   Remove Widget() {is=title K2Node_CustomEvent}
   n37 = Get Target Widgets()
   > n38 = Branch(Condition="integer > integer"(A=Length(TargetArray=@n37), B=1))
   n39 = Last Index!(TargetArray:Array of EUW Emote Live Retarget Child Widget Object References=@n37) -> (ReturnValue=0)
   n40 = Get Target Widgets()
   > RemoveFromParent(self="Get (a copy)"(Array=@n40, Dimension 1=@n39))
   > Remove Index!(execute, TargetArray:Array of EUW Emote Live Retarget Child Widget Object References=@n40, IndexToRemove:int=@n39) -> (then)
   > SetIsEnabled(self=Get AddWidgetbutton(), bInIsEnabled="integer < integer"(B=Get max_targets() -> (max_targets=0)))
   SetIsEnabled(execute=@n38.else, self=Get SubWidgetbutton(), bInIsEnabled=false)
   n41 = CID Updated() {is=title K2Node_CustomEvent}
   n42 = Category Changed!() -> (OutputDelegate, then, SelectedItem:String, SelectionType:ESelectInfo Enum) {is=title K2Node_CustomEvent}
   n43 = Get Current Project(self=Get Anim Browser())
   > n44 = Branch(Condition="Equal, Case Insensitive (String)"(A=GetDisplayName(Object=@n43), B=AnimationBrowser_Companions_Project))
   > Set Companions Project(Companions Project:bool=true) -> (Output_Get:bool)
   > n45 = Execute Python Script!(
    execute
    PythonScript="import unreal\r\n\r\nasset_registry = unreal.AssetRegistryHelpers.get_asset_registry()\r\nall_assets_data = asset_registry.get_assets_by_path(\r\n    \"/CosmeticCompanions/\",\r\n    recursive=True\r\n)\r\n\r\n\r\nfor asset_data in all_assets_data:\r\n    found_asset = ''\r\n    if str(asset_data.asset_name).lower() == search_asset.lower():\r\n        found_asset = asset_data\r\n        loaded_asset = unreal.AssetData.get_asset(found_asset)\r\n        if isinstance(loaded_asset, unreal.CosmeticCompanionItemDefinition):\r\n            found_asset = loaded_asset\r\n            print(found_asset)\r\n            break"
    search_asset:String=Append!(A=Companion_, B=@n42.SelectedItem) -> (ReturnValue)
   ) -> (then, ReturnValue, found_asset:Object Reference) {inputs [search_asset], outputs [found_asset]}
   > n46 = Cast To CosmeticCompanionItemDefinition!(execute, Object=@n45.found_asset) -> (then, CastFailed, AsCosmetic Companion Item Definition:Cosmetic Companion Item Definition Object Reference) {is=title K2Node_DynamicCast}
   > n47 = Set Parent CID(Parent CID=@n46)
   CID Updated(execute=[@n47.then, @n50.then]) {is=title K2Node_CallFunction}
   n48 = Branch(execute=@n44.else, Condition=Get Companions Project() -> (Companions Project:bool))
   > Set Parent CID(Parent CID=/BRCosmetics/Athena/Items/Cosmetics/Characters/CID_A_406_Athena_Commando_M_RebirthFresh)
   > n49 = Set Parent Control Rig(Parent Control Rig=/EpicControlRig/Templates/MR_Biped_v02)
   n50 = Set Companions Project(execute=[@n49.then, @n48.else], Companions Project=false)
   Main Process Event() {is=title K2Node_CustomEvent}
   > n51 = Sequence!(execute) -> (
    then_0
    then_1
    then_2:Exec
    then_3:Exec
    then_4:Exec
    then_5:Exec
    then_6:Exec
    then_7:Exec
    then_8:Exec
    then_9:Exec
    then_10:Exec
    then_11:Exec
    then_12:Exec
   )
   n52 = GetCurrentLevelSequence(execute=[@n51.then_0, @n53.Is Not Valid])
   > n53 = Is Valid(InputObject=@n52)
   > Set Current Level Sequence(Current Level Sequence:Level Sequence Soft Object Reference=Make Soft Reference!(Input:Level Sequence Object Reference=@n52) -> (Output:Level Sequence Soft Object Reference)) -> (Output_Get:Level Sequence Soft Object Reference)
   > Clear(TargetArray:Array of Skeletal Mesh Component Object References=Get Target Skeletal Mesh Components() -> (Target Skeletal Mesh Components:Array of Skeletal Mesh Component Object References))
   > Clear(TargetArray:Array of Athena Character Item Definition Object References=Get Target CIDs!() -> (Target CIDs))
   > Clear(TargetArray:Array of IKRetargeter Object References=Get Target IKRetargets() -> (Target IKRetargets:Array of IKRetargeter Object References))
   > Clear!(
    execute
    TargetMap:Map of Skeletal Mesh Component Object References to Movie Scene Binding Proxy Structures=Get Local Bound Components() -> (Local Bound Components:Map of Skeletal Mesh Component Object References to Movie Scene Binding Proxy Structures)
   ) -> (then) {is=title K2Node_CallFunction}
   > Clear(TargetArray:Array of Skeletal Mesh Object References=Get Children Companion Skeletal Mesh() -> (Children Companion Skeletal Mesh:Array of Skeletal Mesh Object References)) {is=title K2Node_CallArrayFunction}
   n54 = Branch(execute=@n51.then_1, Condition=Get replace())
   > Set Blueprint Proxy(Blueprint Proxy:Movie Scene Binding Proxy Structure=Get binding()) -> (Output_Get:Movie Scene Binding Proxy Structure)
   n55 = Convert to Spawnable!(
    execute=@n54.else
    self
    Input Level Sequence:Level Sequence Object Reference=Resolve Soft Reference(Input=Get Current Level Sequence())
    ActorClass:Actor Class Reference=&BP_LiveRetarget_C
   ) -> (then, "Output Prorxy ":Movie Scene Binding Proxy Structure) {is=title K2Node_CallFunction}
   > Set Blueprint Proxy(Blueprint Proxy=@n55)
   n56 = Branch(execute=@n51.then_2, Condition=Get replace())
   Branch(execute=@n56.else, Condition=IsChecked(self=Get IncludeSourcecbox()))
   > Set Animation!(
    execute
    self
    Input Blueprint Proxy:Movie Scene Binding Proxy Structure=Get Blueprint Proxy() -> (Blueprint Proxy:Movie Scene Binding Proxy Structure)
    Target Animation:Anim Sequence Object Reference=Get Source Animation() -> (Source Animation:Anim Sequence Object Reference)
    Input Level Sequence:Level Sequence Object Reference=Resolve Soft Reference(Input=Get Current Level Sequence())
   ) -> (then) {is=title K2Node_CallFunction}
   n57 = Set Parent CID and get parent BP ref!(
    execute=@n51.then_3
    self
    Input Current Level Sequence:Movie Scene Sequence Object Reference=Resolve Soft Reference(Input=Get Current Level Sequence())
    Input Proxy:Movie Scene Binding Proxy Structure=Get Blueprint Proxy()
    Input Parent CID:Object Reference=Get Parent CID()
   ) -> (then, TargetRetargetClassBP:BP Live Retarget Object Reference) {is=title K2Node_CallFunction}
   > Set TargetRetargetClassBP(TargetRetargetClassBP:BP Live Retarget Object Reference=@n57) -> (Output_Get:BP Live Retarget Object Reference)
   n58 = For Each Loop(Exec=@n51.then_4, Array:Array of EUW Emote Live Retarget Child Widget Object References=Get Target Widgets()) -> (Array Element:EUW Emote Live Retarget Child Widget Object Reference)
   > n59 = Branch(Condition=IsChecked(self=Get IncludeTargetCbox(self=@n58)))
   > n60 = Cast To AthenaCharacterItemDefinition!(
    execute
    Object=Get CID!(self:EUW Emote Live Retarget Child Widget Object Reference=@n58) -> (CID:Fort Account Item Definition Object Reference)
   ) -> (then, CastFailed, AsAthena Character Item Definition:Athena Character Item Definition Object Reference) {is=title K2Node_DynamicCast}
   > n61 = Add(
    TargetArray:Array of Athena Character Item Definition Object References=Get Target CIDs()
    NewItem:"Athena Character Item Definition Object Reference (by ref)"=@n60
   )
   Add(
    execute=[@n61.then, @n64.then]
    TargetArray:Array of IKRetargeter Object References=Get Target IKRetargets()
    NewItem:"IKRetargeter Object Reference (by ref)"=Get IK Retargeter!(self:EUW Emote Live Retarget Child Widget Object Reference=@n58) -> (IK Retargeter:IKRetargeter Object Reference)
   )
   Add(
    execute=@n60.CastFailed
    TargetArray:Array of Athena Character Item Definition Object References=Get Target CIDs()
    NewItem:"Athena Character Item Definition Object Reference (by ref)"
   )
   n62 = Get Companion Skeletal Mesh!(self:EUW Emote Live Retarget Child Widget Object Reference=@n58) -> (Companion Skeletal Mesh:Skeletal Mesh Object Reference)
   > n63 = Is Valid(InputObject=@n62)
   Add(
    execute=@n63.Is Not Valid
    TargetArray:Array of Skeletal Mesh Object References=Get Children Companion Skeletal Mesh()
    NewItem:"Skeletal Mesh Object Reference (by ref)"
   )
   n64 = Add(execute=@n63.Is Valid, TargetArray=Get Children Companion Skeletal Mesh(), NewItem=@n62)
   n65 = Set Data for none active components!(
    execute=@n59.else
    self
    input index:int=@n58.Array Index
    input bp:BP Live Retarget Object Reference=Get TargetRetargetClassBP() -> (TargetRetargetClassBP:BP Live Retarget Object Reference)
   ) -> (
    then
    output cid:Fort Account Item Definition Object Reference
    output IK retarget:IKRetargeter Object Reference
    output skeletal mesh:Skeletal Mesh Object Reference
   ) {is=title K2Node_CallFunction}
   > n66 = Sequence!(execute) -> (then_0, then_1)
   > n67 = Cast To AthenaCharacterItemDefinition(Object=@n65)
   > Add(
    TargetArray:Array of Athena Character Item Definition Object References=Get Target CIDs()
    NewItem:"Athena Character Item Definition Object Reference (by ref)"=@n67
   )
   Add(
    execute=@n66.then_1
    TargetArray:Array of IKRetargeter Object References=Get Target IKRetargets()
    NewItem:"IKRetargeter Object Reference (by ref)"=@n65.output IK retarget
   )
   > Add(
    TargetArray:Array of Skeletal Mesh Object References=Get Children Companion Skeletal Mesh()
    NewItem:"Skeletal Mesh Object Reference (by ref)"=@n65.output skeletal mesh
   )
   SetEditorProperty(
    execute=@n51.then_5
    Object=Get TargetRetargetClassBP()
    PropertyName=Target CIDs
    PropertyValue:Array of Athena Character Item Definition Object References=Get Target CIDs()
   )
   > SetEditorProperty(Object=Get TargetRetargetClassBP(), PropertyName=IKRetargets, PropertyValue:Array of IKRetargeter Object References=Get Target IKRetargets())
   > SaveDefaultSpawnableState!(
    execute
    self:Level Sequence Editor Subsystem Object Reference=Get LevelSequenceEditorSubsystem()
    ObjectBinding:"Movie Scene Binding Proxy Structure (by ref)"=Get Blueprint Proxy()
   ) -> (then)
   Reroute Node(InputPin=@n51.then_6)
   n68 = GetChildPossessables!(execute=@n51.then_8, InBinding:"Movie Scene Binding Proxy Structure (by ref)"=Get Blueprint Proxy()) -> (then, ReturnValue:Array of Movie Scene Binding Proxy Structures)
   > n69 = For Each Loop(Array:Array of Movie Scene Binding Proxy Structures=@n68) -> (Array Element:Movie Scene Binding Proxy Structure)
   > n70 = GetBindingID(Sequence=Resolve Soft Reference(Input=Get Current Level Sequence()), InBinding=@n69)
   > n71 = For Each Loop(Array:Array of Object References=GetBoundObjects(ObjectBinding=@n70)) -> (Array Element:Object Reference)
   > n72 = Cast To SkeletalMeshComponent(Object=@n71)
   > Add!(
    execute
    TargetMap:Map of Skeletal Mesh Component Object References to Movie Scene Binding Proxy Structures=Get Local Bound Components()
    Key:"Skeletal Mesh Component Object Reference (by ref)"=@n72
    Value:"Movie Scene Binding Proxy Structure (by ref)"=@n69
   ) -> (then) {is=title K2Node_CallFunction}
   SetEditorProperty(
    execute=@n51.then_7
    Object=Get TargetRetargetClassBP()
    PropertyName=Target Companion Skeletal Meshes
    PropertyValue:Array of Skeletal Mesh Object References=Get Children Companion Skeletal Mesh()
   )
   > SaveDefaultSpawnableState(self=Get LevelSequenceEditorSubsystem(), ObjectBinding=Get Blueprint Proxy())
   n73 = For Each Loop(Exec=@n51.then_10, Array:Array of EUW Emote Live Retarget Child Widget Object References=Get Target Widgets()) -> (Array Element:EUW Emote Live Retarget Child Widget Object Reference)
   > n74 = Execute Python Script!(execute, PythonScript="character = f\"{chr(ord('@')+(index+1))}\"", index:int=@n73.Array Index) -> (then, ReturnValue, character:String) {inputs [index], outputs [character]}
   > Branch(Condition=IsChecked(self=Get IncludeTargetCbox(self=@n73)))
   n75 = Get Text(self=Get SuffixTextBox!(self:EUW Emote Live Retarget Child Widget Object Reference=@n73) -> (SuffixTextBox:Editor Utility Editable Text Box Object Reference))
   n76 = Get ItemName!(self:Item Definition Base Object Reference=Get CID(self=@n73)) -> (ItemName:Text)
   > Set Posseable and Bind Control Rig!(
    execute
    self
    Input Level Sequence:Level Sequence Object Reference=Resolve Soft Reference(Input=Get Current Level Sequence())
    ObjectToPossess:Skeletal Mesh Component Object Reference="Get (a copy)"(Array:Array of Skeletal Mesh Component Object References=Get Target Skeletal Mesh Components(), Dimension 1=@n73.Array Index) -> (Output:Skeletal Mesh Component Object Reference)
    Target Control Rig:Control Rig Blueprint Object Reference=Get Control Rig!(self:EUW Emote Live Retarget Child Widget Object Reference=@n73) -> (Control Rig:Control Rig Blueprint Object Reference)
    Input Local Bound Components:Map of Skeletal Mesh Component Object References to Movie Scene Binding Proxy Structures=Get Local Bound Components()
    Rename Component:bool=true
    Component Rename:String=Append!(A=@n74.character, B=__, C:String="To String (Text)"(InText=@n76), D:String=_, E:String="To String (Text)"(InText=@n75)) -> (ReturnValue)
    only_set_possesable:bool=false
    layered rig:bool=true
   ) -> (then) {is=title K2Node_CallFunction}
   Call E Export Animations!(execute:Exec=@n51.then_11, self:EUW Emote Live Retarget Object Reference) -> (then:Exec) {is=title K2Node_CallDelegate}
   > Branch(Condition=IsChecked(self=Get checkoutLS_cbox() -> (checkoutLS_cbox:Editor Utility Check Box Object Reference)))
   n77 = Resolve Soft Reference(Input=Get Current Level Sequence())
   > CheckoutLoadedAsset!(execute, AssetToCheckout:Object Reference=@n77) -> (then, ReturnValue:bool)
   > SaveLoadedAsset(AssetToSave=@n77, bOnlyIfIsDirty=false)
   > Set replace(replace=false)
   > Set CheckedState!(
    execute
    CheckedState:ECheckBoxState Enum=Checked
    self:Check Box Object Reference=Get LayeredMasterRigCbox() -> (LayeredMasterRigCbox:Editor Utility Check Box Object Reference)
   ) -> (then, Output_Get:ECheckBoxState Enum=Unchecked)
   n78 = GetBindingID(execute=@n51.then_9, Sequence=Resolve Soft Reference(Input=Get Current Level Sequence()), InBinding=Get Blueprint Proxy())
   n79 = "Get (a copy)"(Array:Array of Object References=GetBoundObjects(ObjectBinding=@n78)) -> (Output:Object Reference)
   > Set bound ID(bound ID:Object Reference=@n79) -> (Output_Get:Object Reference)
   > n80 = GetAllActorsOfClass!(execute, WorldContextObject:Object Reference=@n79, ActorClass:Actor Class Reference=/Script/Engine.SkeletalMeshActor) -> (then, OutActors:Array of Skeletal Mesh Actor Object References)
   > n81 = For Each Loop(Array:Array of Skeletal Mesh Actor Object References=@n80) -> (Array Element:Skeletal Mesh Actor Object Reference)
   > Cast To BP_LiveRetarget(Object=@n81)
   n82 = Get SkeletalMeshComponent!(self:Skeletal Mesh Actor Object Reference=@n81) -> (SkeletalMeshComponent:Skeletal Mesh Component Object Reference)
   > Set New Actor Skeletal Mesh Component(New Actor Skeletal Mesh Component:Skeletal Mesh Component Object Reference=@n82) -> (Output_Get:Skeletal Mesh Component Object Reference)
   > n83 = Branch(Condition=IsChecked(self=Get IncludeSourcecbox()))
   > n84 = Set Posseable and Bind Control Rig(
    Input Level Sequence=Resolve Soft Reference(Input=Get Current Level Sequence())
    ObjectToPossess=@n82
    Target Control Rig=Get Parent Control Rig() -> (Parent Control Rig:Control Rig Blueprint Object Reference)
    Input Local Bound Components=Get Local Bound Components()
    Rename Component=false
    only_set_possesable=Get replace()
    layered rig=IsChecked(self=Get LayeredMasterRigCbox())
   )
   n85 = For Each Loop(
    Exec=[@n84.then, @n89.then]
    Array:Array of Scene Component Object References=GetChildrenComponents!(
     self:Scene Component Object Reference=Get New Actor Skeletal Mesh Component() -> (New Actor Skeletal Mesh Component:Skeletal Mesh Component Object Reference)
     bIncludeAllDescendants:bool=false
    ) -> (Children:Array of Scene Component Object References)
   ) -> (Array Element:Scene Component Object Reference)
   > n86 = Cast To SkeletalMeshComponent(Object=@n85)
   n87 = GetDisplayName(Object=@n85)
   > n88 = Branch(Condition=Contains(SearchIn=@n87, Substring=.NODE_AddSkeletalMeshComponent-))
   Branch(execute=@n88.else, Condition="Equal (Object)"(A=GetOwner(self=@n86), B=Get bound ID() -> (bound ID:Object Reference)))
   > Add(
    TargetArray:Array of Skeletal Mesh Component Object References=Get Target Skeletal Mesh Components()
    NewItem:"Skeletal Mesh Component Object Reference (by ref)"=@n86
   ) {is=title K2Node_CallArrayFunction}
   n89 = SaveDefaultSpawnableState(execute=@n83.else, self=Get LevelSequenceEditorSubsystem(), ObjectBinding=Get Blueprint Proxy())
   n90 = clear all() {is=title K2Node_CustomEvent}
   n91 = set new ani m paths!() -> (OutputDelegate, then, AnimSequence Path:String, LevelSequence Path:String, Animatable Creature:Animatable Creature Object Reference) {is=title K2Node_CustomEvent}
   > Set Animation Path(Animation Path=@n91.AnimSequence Path)
   > n92 = Set Anim Sequence(Anim Sequence=@n91.LevelSequence Path)
   E_ContextMenu_Pressed() {is=title K2Node_CustomEvent}
   n93 = Get Anim Browser()
   > SetRenderTransform!(
    execute
    self:Widget Object Reference=GetParent!(self:Widget Object Reference=Get EUW_ContextMenu!(self:Animation Browser Object Reference=@n93) -> (EUW_ContextMenu:EUW Context Menu Object Reference)) -> (ReturnValue:Panel Widget Object Reference)
    InTransform_Translation:Vector 2D Structure=(X=-300,Y=300)
    InTransform_Scale:Vector 2D Structure=(X=1,Y=1)
    InTransform_Shear:Vector 2D Structure=(X=0,Y=0)
    InTransform_Angle:float=0
   ) -> (then)
   Event Construct()
   > n94 = Sequence!(execute) -> (then_0, then_1, then_2:Exec, then_3:Exec, then_4:Exec)
   > n95 = For Loop with Break!(Execute:Exec, FirstIndex:int, LastIndex:int=int - int(A=Get Set Targets() -> (Set Targets:int=0)), Break:Exec=@n97.else) -> (LoopBody:Exec, Index:int, Completed:Exec) {is=title K2Node_MacroInstance}
   n96 = Get max_targets() -> (max_targets=0)
   > n97 = Branch(Condition="integer < integer"(A=@n95, B=int - int(A=@n96)))
   > n98 = Add Child Widget()
   > Set Widget Text and Color(Widget=@n98, index=@n95)
   n99 = Create Widget(execute=@n94.then_1, Class=/AnimInEdTools/AnimationBrowser/Animation_Browser.Animation_Browser_C) -> (ReturnValue:Animation Browser Object Reference)
   > Set Anim Browser(Anim Browser:Animation Browser Object Reference=@n99) -> (Output_Get:Animation Browser Object Reference)
   > AddChild(self=Get AnimBrowserWidget() -> (AnimBrowserWidget:Named Slot Object Reference), Content=@n99)
   > Unbind all Events from On Clicked!(
    execute:Exec
    self:Button Object Reference=Get AddAnimationButton!(self:Animation Browser Object Reference=Get Anim Browser()) -> (AddAnimationButton:Button Object Reference)
   ) -> (then:Exec) {is=title K2Node_ClearDelegate}
   > SetVisibility(self=Get AddAnimationButton(self=Get Anim Browser()), InVisibility=Hidden)
   Set Human Defaults(execute=@n94.then_3) {is=title K2Node_CallFunction}
   Assign On Pressed!(
    execute:Exec=@n94.then_4
    self:Button Object Reference=Get SettingsButton!(self:Animation Browser Object Reference=Get Anim Browser()) -> (SettingsButton:Button Object Reference)
    Delegate:"Delegate (by ref)"=Create Event()
   ) -> (then:Exec) {is=title K2Node_AssignDelegate}
   Bind Event to E Bind Control Rig(execute=@n94.then_2, self:EUW Emote Live Retarget Object Reference, Delegate=Create Event()) {is=title K2Node_AddDelegate}
   > Bind Event to E Export Animations(self:EUW Emote Live Retarget Object Reference, Delegate=Create Event()) {is=title K2Node_AddDelegate}
   > Bind Event to E Main Process(self:EUW Emote Live Retarget Object Reference, Delegate=Create Event()) {is=title K2Node_AddDelegate}
   > Bind Event to On Released(self:Button Object Reference=Get SettingsButton(self=Get Anim Browser()), Delegate=Create Event()) {is=title K2Node_AddDelegate}
   > Bind Event to E Remove Widget(self:EUW Emote Live Retarget Object Reference, Delegate=Create Event()) {is=title K2Node_AddDelegate}
   > Bind Event to On Selection Changed(self:"Searchable Combo Box (String) Object Reference"=Get Asset_ComboBox(self=Get Anim Browser()), Delegate=Create Event()) {is=title K2Node_AddDelegate}
   > Bind Event to On Item Clicked(
    self:List View Object Reference=Get ListView!(self:Animation Browser Object Reference=Get Anim Browser()) -> (ListView:List View Object Reference)
    Delegate=Create Event()
   ) {is=title K2Node_AddDelegate}
   Event PreConstruct()
   > Sequence!(execute) -> (then_0, then_1)
   > SetObject(
    self=[
     Get ParentCIDView() -> (ParentCIDView:Single Property View Object Reference)
     Get ParentCrigView() -> (ParentCrigView:Single Property View Object Reference)
     Get ParentAnimationSinglePropertyView() -> (ParentAnimationSinglePropertyView:Single Property View Object Reference)
    ]
    NewObject=Self-Reference()
   )
   > SetPropertyName!(execute, self:Single Property View Object Reference=Get ParentAnimationSinglePropertyView(), NewPropertyName:Name=Source Animation) -> (then)
   > SetPropertyName(self=Get ParentCIDView(), NewPropertyName=Parent CID)
   > SetPropertyName(self=Get ParentCrigView(), NewPropertyName=Parent Control Rig)
   n100 = For Each Loop(Array:Array of EUW Emote Live Retarget Child Widget Object References=Get Target Widgets()) -> (Array Element:EUW Emote Live Retarget Child Widget Object Reference)
   n101 = Get Companion Skeletal Mesh(self=@n100)
   > n102 = Is Valid(InputObject=@n101)
   Add(
    execute=@n102.Is Not Valid
    TargetArray:Array of Skeletal Mesh Object References=Get Children Companion Skeletal Mesh()
    NewItem:"Skeletal Mesh Object Reference (by ref)"
   )
   Branch(execute=@n102.Is Valid, Condition="integer <= integer"(A=Length(TargetArray:Array of Wildcards), B=@n100.Array Index))
   > Add(TargetArray=Get Children Companion Skeletal Mesh(), NewItem=@n101)
   n103 = GetDisplayName(Object=Get Current Project(self=Get Anim Browser()))
   Set Project(Project:String=@n103) -> (Output_Get:String)
   > Branch(Condition="Equal Exactly (String)"(A=@n103, B=AnimationBrowser_BREmotes_Project))
   > n104 = Sequence!(execute) -> (then_0, then_1, then_2:Exec)
   n105 = Get Target Widgets()
   > While Loop(Condition="integer < integer"(A=Length(TargetArray:Array of EUW Emote Live Retarget Child Widget Object References=@n105), B=Get Set Targets() -> (Set Targets=0)))
   > Add Widget()
   Set Human Defaults(execute=@n104.then_1)
   Bind Event to On Clicked(self=Get AddAnimationButton(self=Get Anim Browser()), Delegate=Create Event())
   Contains(SearchIn=@n87, Substring=.target)
   Project Settings Changed() {is=title K2Node_CustomEvent}
   Get SuffixTextBox(self=@n23)
  ]}
 Generate Animations from Dialog and call Main Process {nodes [
   Generate Animations from Dialog and call Main Process() {is=title K2Node_FunctionEntry}
   > n1 = Validate UI Input!(execute, self:EUW Animation Dialog Object Reference=Get EUW Anim Dialog()) -> (then, Is Valid:bool)
   > Branch(Condition=@n1)
   > n2 = Generate Anim and Level Sequence Paths!(execute, self:EUW Animation Dialog Object Reference=Get EUW Anim Dialog()) -> (
    then
    Anim Sequence:String
    Level Sequence:String
    Animatable Creatures:Animatable Creature Object Reference
    Animation Path:String
    Source Path:String
   )
   > Set Anim Sequence(Anim Sequence=@n2)
   > Set Animation Path(Animation Path=@n2.Animation Path)
   > n3 = Construct Object from Class(Class=/Script/LevelSequenceEditor.LevelSequenceFactoryNew) -> (ReturnValue:Level Sequence Factory New Object Reference)
   > n4 = GetAssetTools()
   > n5 = CreateAsset(
    self=@n4
    AssetName=GetCleanFilename!(InPath:String=@n2.Level Sequence) -> (ReturnValue:String)
    PackagePath=GetPath(InPath=@n2.Level Sequence)
    AssetClass=/Script/LevelSequence
    Factory=@n3
   )
   > n6 = Cast To LevelSequence!(execute, Object=@n5) -> (then, CastFailed, AsLevel Sequence:Level Sequence Object Reference) {is=title K2Node_DynamicCast}
   > n7 = Set Current Level Sequence(Current Level Sequence=Make Soft Reference(Input=@n6))
   > OpenLevelSequence!(execute, LevelSequence:Level Sequence Object Reference=Resolve Soft Reference(Input=@n7)) -> (then, ReturnValue:bool)
   > Call E Main Process()
  ]}
 Get Animation {nodes [
   Get Animation!() -> (then, Input Sequence:Movie Scene Sequence Object Reference) {is=title K2Node_FunctionEntry}
   > n1 = Sequence!(execute) -> (then_0, then_1)
   > Is Valid(InputObject=Get Input Sequence() -> (Input Sequence:Movie Scene Sequence Object Reference))
   > n2 = GetBindings(Sequence=Get Input Sequence())
   > n3 = For Each Loop(Array:Array of Movie Scene Binding Proxy Structures=@n2) -> (Array Element:Movie Scene Binding Proxy Structure)
   > Set Local Binding(Local Binding=@n3)
   > n4 = GetBindingID(Sequence=Get Input Sequence(), InBinding=Get Local Binding())
   > n5 = For Each Loop(Array:Array of Object References=GetBoundObjects(ObjectBinding=@n4)) -> (Array Element:Object Reference)
   > Cast To BP_LiveRetarget(Object=@n5)
   > n6 = GetTracks!(execute, InBinding:"Movie Scene Binding Proxy Structure (by ref)"=Get Local Binding()) -> (then, ReturnValue:Array of Movie Scene Track Object References)
   > n7 = For Each Loop(Array:Array of Movie Scene Track Object References=@n6) -> (Array Element:Movie Scene Track Object Reference)
   > n8 = Cast To MovieSceneSkeletalAnimationTrack!(execute, Object=@n7) -> (then, CastFailed, AsMovie Scene Skeletal Animation Track:Movie Scene Skeletal Animation Track Object Reference) {is=title K2Node_DynamicCast}
   > n9 = GetSections!(execute, Track:Movie Scene Track Object Reference=@n8) -> (then, ReturnValue:Array of Movie Scene Section Object References)
   > n10 = For Each Loop(Array:Array of Movie Scene Section Object References=@n9) -> (Array Element:Movie Scene Section Object Reference)
   > n11 = Cast To MovieSceneSkeletalAnimationSection!(execute, Object=@n10) -> (then, CastFailed, AsMovie Scene Skeletal Animation Section:Movie Scene Skeletal Animation Section Object Reference) {is=title K2Node_DynamicCast}
   > Set Local Animation(Local Animation:Anim Sequence Base Object Reference=Get!(self:Movie Scene Skeletal Animation Section Object Reference=@n11) -> (
     Params_Animation:Anim Sequence Base Object Reference
     Params_FirstLoopStartFrameOffset:Frame Number Structure
     Params_StartFrameOffset:Frame Number Structure
     Params_EndFrameOffset:Frame Number Structure
     Params_PlayRate:Movie Scene Time Warp Variant Structure
     Params_bReverse:bool=~
     Params_SlotName:Name
     Params_MirrorDataTable:Mirror Data Table Object Reference
     Params_bSkipAnimNotifiers:bool=~
     Params_bForceCustomMode:bool=~
     Params_SwapRootBone:ESwapRootBone Enum
     Params_bLinearPlaybackWhenScaled:bool=~
    )) -> (Output_Get:Anim Sequence Base Object Reference)
   > Set Local Animation Section(Local Animation Section:Movie Scene Section Object Reference=@n10) -> (Output_Get:Movie Scene Section Object Reference)
   Return Node!(
    execute=@n1.then_1
    Ouput Params Animation:Anim Sequence Base Object Reference=Get Local Animation() -> (Local Animation:Anim Sequence Base Object Reference)
    Output Animation Section:Movie Scene Section Object Reference=Get Local Animation Section() -> (Local Animation Section:Movie Scene Section Object Reference)
   )
  ]}
 Get Linked Animation Path {nodes [
   Get Linked Animation Path!() -> (then, InLevelSequence:Level Sequence Object Reference, MovieSceneBindingProxy:"Movie Scene Binding Proxy Structure (by ref)") {is=title K2Node_FunctionEntry}
   > n1 = Sequence()
   > n2 = GetParent!(
    execute:Exec
    InBinding:"Movie Scene Binding Proxy Structure (by ref)"=Get MovieSceneBindingProxy() -> (MovieSceneBindingProxy:"Movie Scene Binding Proxy Structure (by ref)")
   ) -> (then:Exec, ReturnValue:Movie Scene Binding Proxy Structure)
   > n3 = GetPossessedObjectClass(InBinding=@n2)
   > Branch(Condition="Equal (Class)"(A=@n3, B=&BP_LiveRetarget_C))
   > n4 = GetAnimSequenceLinkFromLevelSequence!(execute, InLevelSequence:Level Sequence Object Reference=Get InLevelSequence() -> (InLevelSequence:Level Sequence Object Reference)) -> (then, ReturnValue:Level Sequence Anim Sequence Link Object Reference)
   > n5 = For Each Loop!(
    Exec
    Array:Array of Level Sequence Anim Sequence Link Item Structures=Get AnimSequenceLinks!(self:Level Sequence Anim Sequence Link Object Reference=@n4) -> (AnimSequenceLinks:Array of Level Sequence Anim Sequence Link Item Structures)
   ) -> (
    LoopBody
    Array Element_SkelTrackGuid:Guid Structure
    Array Element_PathToAnimSequence:Soft Object Path Structure
    Array Element_bAutoBake:bool=~
    Array Element_bExportTransforms:bool=~
    Array Element_bExportMorphTargets:bool=~
    Array Element_bExportAttributeCurves:bool=~
    Array Element_bExportMaterialCurves:bool=~
    Array Element_bSkipCurvesWithZeroValue:bool=~
    Array Element_bSetRetargetSourceAsset:bool=~
    Array Element_Interpolation:EAnimInterpolationType Enum
    Array Element_CurveInterpolation:ERichCurveInterpMode Enum
    Array Element_bRecordInWorldSpace:bool=~
    Array Element_bEvaluateAllSkeletalMeshComponents:bool=~
    Array Element_IncludeAnimationNames:Array of Strings
    Array Element_ExcludeAnimationNames:Array of Strings
    Array Element_bRemoveExcludedCurves:bool=~
    Array Element_WarmUpFrames:Frame Number Structure
    Array Element_DelayBeforeStart:Frame Number Structure
    Array Element_bUseCustomTimeRange:bool=~
    Array Element_CustomStartFrame:Frame Number Structure
    Array Element_CustomEndFrame:Frame Number Structure
    Array Element_CustomDisplayRate:Frame Rate Structure
    Array Element_bUseCustomFrameRate:bool=~
    Array Element_CustomFrameRate:Frame Rate Structure
    Array Index
    Completed
   )
   > Branch(Condition=IsEmpty(InString=Get found animation path() -> (found animation path:String)))
   > n6 = FindBindingById!(execute, Sequence:Movie Scene Sequence Object Reference=Get InLevelSequence(), BindingId:Guid Structure=@n5) -> (then, ReturnValue_BindingID:Guid Structure, ReturnValue_Sequence:Movie Scene Sequence Object Reference)
   n7 = Get MovieSceneBindingProxy()
   > Branch(Condition="Equal (Guid)"(
     A:"Guid Structure (by ref)"=@n6
     B:"Guid Structure (by ref)"=Break Movie Scene Binding Proxy!(MovieSceneBindingProxy:"Movie Scene Binding Proxy Structure (by ref)"=@n7) -> (BindingID:Guid Structure, Sequence:Movie Scene Sequence Object Reference) {is=title K2Node_BreakStruct}
    ) -> (ReturnValue:bool=~) {is=title K2Node_PromotableOperator})
   > Set found animation path(found animation path:String=BreakSoftObjectPath(InSoftObjectPath=@n5.Array Element_PathToAnimSequence)) -> (Output_Get:String)
   Return Node!(execute=@n1.then_1, Output Animation Path:String=Get found animation path())
  ]}
 Link New Paths {nodes [Link New Paths() {is=title K2Node_FunctionEntry}]}
 Set Animation {nodes [
   Set Animation!() -> (
    then
    Input Blueprint Proxy:Movie Scene Binding Proxy Structure
    Target Animation:Anim Sequence Object Reference
    Input Level Sequence:Level Sequence Object Reference
   ) {is=title K2Node_FunctionEntry}
   > n1 = Sequence()
   > n2 = Get Animation(Input Sequence=Get Input Level Sequence()) {is=title K2Node_CallFunction}
   > n3 = Is Valid(InputObject=@n2.Output Animation Section)
   n4 = AddTrack!(
    execute=@n3.Is Not Valid
    InBinding:"Movie Scene Binding Proxy Structure (by ref)"=Get Input Blueprint Proxy() -> (Input Blueprint Proxy:Movie Scene Binding Proxy Structure)
    TrackType:Movie Scene Track Class Reference=/Script/MovieSceneTracks.MovieSceneSkeletalAnimationTrack
   ) -> (then, ReturnValue:Movie Scene Track Object Reference)
   > n5 = AddSection!(execute, Track:Movie Scene Track Object Reference=@n4) -> (then, ReturnValue:Movie Scene Section Object Reference)
   > Set Local Animation Section(Local Animation Section=@n5)
   Set Local Animation Section(execute=@n3.Is Valid, Local Animation Section=@n2.Output Animation Section)
   SetStartFrame!(execute=@n1.then_1, Section:Movie Scene Section Object Reference=Get Local Animation Section(), StartFrame:int=0) -> (then)
   > SetEndFrame!(
    execute
    Section:Movie Scene Section Object Reference=Get Local Animation Section()
    EndFrame:int=GetNumFrames!(AnimationSequenceBase:Anim Sequence Base Object Reference=Get Target Animation() -> (Target Animation:Anim Sequence Object Reference)) -> (NumFrames:int=0)
   ) -> (then)
   > n6 = Cast To MovieSceneSkeletalAnimationSection(Object=Get Local Animation Section())
   > n7 = Set!(
    execute
    Params_Animation:Anim Sequence Base Object Reference=Get Target Animation()
    Params_FirstLoopStartFrameOffset:Frame Number Structure=(Value=0)
    Params_StartFrameOffset:Frame Number Structure=(Value=0)
    Params_EndFrameOffset:Frame Number Structure=(Value=0)
    Params_PlayRate:Movie Scene Time Warp Variant Structure="EMovieSceneTimeWarpType::FixedPlayRate(PlayRate=1)"
    Params_bReverse:bool=False
    Params_SlotName:Name=DefaultSlot
    Params_MirrorDataTable:Mirror Data Table Object Reference
    Params_bSkipAnimNotifiers:bool=False
    Params_bForceCustomMode:bool=False
    Params_SwapRootBone:ESwapRootBone Enum=SwapRootBone_None
    Params_bLinearPlaybackWhenScaled:bool=False
    self:Movie Scene Skeletal Animation Section Object Reference=@n6
   ) -> (then, Output_Get:Movie Scene Skeletal Animation Params Structure)
   > Set members in Movie Scene Skeletal Animation Params!(execute:Exec, StructRef:"Movie Scene Skeletal Animation Params Structure (by ref)"=@n7) -> (then:Exec, StructOut:"Movie Scene Skeletal Animation Params Structure (by ref)") {is=title K2Node_SetFieldsInStruct}
   > n8 = GetAutoSizeHasEndFrame!(execute, Section:Movie Scene Section Object Reference=@n6) -> (then, ReturnValue:bool)
   > Branch(Condition=@n8)
   > n9 = GetAutoSizeEndFrameSeconds!(execute, Section:Movie Scene Section Object Reference=@n6) -> (then, ReturnValue:float)
   > SetEndFrameSeconds!(execute, Section:Movie Scene Section Object Reference=@n6, EndTime:float=@n9) -> (then)
   > SetPlaybackEndSeconds!(execute, Sequence:Movie Scene Sequence Object Reference=Get Input Level Sequence(), EndTime:float=@n9) -> (then)
  ]}
 Set Data for none active components {nodes [
   Set Data for none active components!() -> (then, input index:int, input bp:BP Live Retarget Object Reference) {is=title K2Node_FunctionEntry}
   > Set local found cid(local found cid:Fort Account Item Definition Object Reference) -> (Output_Get:Fort Account Item Definition Object Reference)
   > n1 = Sequence!(execute) -> (then_0, then_1, then_2:Exec)
   > n2 = Is Valid(InputObject=Get Original Companion Skeletal Mesh!(self:BP Live Retarget Object Reference=Get input bp() -> (input bp:BP Live Retarget Object Reference)) -> (Original Companion Skeletal Mesh:Skeletal Mesh Object Reference))
   n3 = Get input bp()
   Set local found cid(
    execute=@n2.Is Not Valid
    local found cid="Get (a copy)"(
     Array:Array of Athena Character Item Definition Object References=Get Target CIDs!(self:BP Live Retarget Object Reference=@n3) -> (Target CIDs)
     Dimension 1=Get input index() -> (input index:int=0)
    ) -> (Output:Athena Character Item Definition Object Reference)
   )
   n4 = "Get (a copy)"(
    Array:Array of Skeletal Mesh Object References=Get Target Companion Skeletal Meshes!(self:BP Live Retarget Object Reference=Get input bp()) -> (Target Companion Skeletal Meshes:Array of Skeletal Mesh Object References)
    Dimension 1=Get input index() -> (input index=0)
   ) -> (Output:Skeletal Mesh Object Reference)
   Is Valid(exec=@n2.Is Valid, InputObject=@n4)
   > n5 = Execute Python Script!(
    execute
    PythonScript="import unreal\r\n\r\npackage_name = unreal.EditorAssetLibrary.find_asset_data(unreal.EditorAssetLibrary.get_path_name_for_loaded_asset(input_asset)).package_name\r\ndependecy = unreal.AssetRegistryDependencyOptions(include_hard_management_references =True, include_hard_package_references =True,include_searchable_names=True,include_soft_management_references =True,include_soft_package_references=True)\r\n\r\n\r\nfound_item = ''\r\nfor asset in unreal.AssetRegistryHelpers.get_asset_registry().get_referencers(package_name, dependecy):\r\n    item = unreal.load_asset(asset)\r\n    if found_item: \r\n        break\r\n    if isinstance(item, unreal.CustomCharacterPart):\r\n        target_package  = unreal.EditorAssetLibrary.find_asset_data(unreal.EditorAssetLibrary.get_path_name_for_loaded_asset(item)).package_name\r\n        for target in unreal.AssetRegistryHelpers.get_asset_registry().get_referencers(target_package, dependecy):\r\n            item = unreal.load_asset(target)\r\n            if isinstance(item, unreal.CosmeticCompanionItemDefinition):\r\n                found_item = item\r\n"
    input_asset:Skeletal Mesh Object Reference=@n4
   ) -> (then, ReturnValue, found_item:Object Reference) {inputs [input_asset], outputs [found_item]}
   > Is Valid(InputObject=@n5.found_item)
   > Cast To FortAccountItemDefinition!(execute, Object=@n5.found_item) -> (then, CastFailed, AsFort Account Item Definition:Fort Account Item Definition Object Reference) {is=title K2Node_DynamicCast}
   > Set local found skeletal mesh(local found skeletal mesh:Skeletal Mesh Object Reference=@n4) -> (Output_Get:Skeletal Mesh Object Reference)
   n6 = Get input bp()
   Set local IK target(
    execute=@n1.then_1
    local IK target:IKRetargeter Object Reference="Get (a copy)"(
     Array:Array of IKRetargeter Object References=Get IKRetargets!(self:BP Live Retarget Object Reference=@n6) -> (IKRetargets:Array of IKRetargeter Object References)
     Dimension 1=Get input index() -> (input index=0)
    ) -> (Output:IKRetargeter Object Reference)
   ) -> (Output_Get:IKRetargeter Object Reference)
   Return Node!(
    execute=@n1.then_2
    output cid:Fort Account Item Definition Object Reference=Get local found cid() -> (local found cid:Fort Account Item Definition Object Reference)
    output IK retarget:IKRetargeter Object Reference=Get local IK target() -> (local IK target:IKRetargeter Object Reference)
    output skeletal mesh:Skeletal Mesh Object Reference=Get local found skeletal mesh() -> (local found skeletal mesh:Skeletal Mesh Object Reference)
   )
  ]}
 Set Human Defaults {nodes [
   Set Human Defaults() {is=title K2Node_FunctionEntry}
   > n1 = Sequence()
   > n2 = For Each Loop!(Exec, Array:Array of EUW Emote Live Retarget Child Widget Object References=Get Target Widgets()) -> (LoopBody, Array Element:EUW Emote Live Retarget Child Widget Object Reference, Array Index, Completed)
   > Branch(Condition="integer < integer"(A=@n2.Array Index, B=Get Set Targets() -> (Set Targets=0)))
   n3 = Get CID(self=@n2)
   > Branch(Condition="Equal Exactly (String)"(A=GetDisplayName(Object=@n3)))
   > Set CID(
    CID="Get (a copy)"(
     Array:Array of Athena Character Item Definition Object References=Get Default Human CIDs() -> (Default Human CIDs:Array of Athena Character Item Definition Object References)
     Dimension 1=@n2.Array Index
    ) -> (Output:Athena Character Item Definition Object Reference)
    self=@n2
   )
   > CID Changed(self:EUW Emote Live Retarget Child Widget Object Reference=@n2)
   n4 = Get Parent CID()
   Branch(execute=@n1.then_1, Condition="Equal Exactly (String)"(A=GetDisplayName(Object=@n4)))
   > Set Parent CID(Parent CID=&CID_A_406_Athena_Commando_M_RebirthFresh)
   n5 = Get Parent Control Rig()
   Branch(execute=@n1.then_2, Condition="Equal Exactly (String)"(A=GetDisplayName(Object=@n5)))
   > Set Parent Control Rig(Parent Control Rig=&MR_Biped_v02)
  ]}
 Set Parent CID and get parent BP ref {nodes [
   Set Parent CID and get parent BP ref!() -> (
    then
    Input Current Level Sequence:Movie Scene Sequence Object Reference
    Input Proxy:Movie Scene Binding Proxy Structure
    Input Parent CID:Object Reference
   ) {is=title K2Node_FunctionEntry}
   > n1 = Sequence()
   > n2 = GetBindingID(
    Sequence=Get Input Current Level Sequence() -> (Input Current Level Sequence:Movie Scene Sequence Object Reference)
    InBinding=Get Input Proxy() -> (Input Proxy:Movie Scene Binding Proxy Structure)
   )
   > n3 = For Each Loop(Array:Array of Object References=GetBoundObjects(ObjectBinding=@n2)) -> (Array Element:Object Reference)
   > n4 = Cast To BP_LiveRetarget(Object=@n3)
   > Set Class Cast(Class Cast:BP Live Retarget Object Reference=@n4) -> (Output_Get:BP Live Retarget Object Reference)
   > Branch(Condition=IsChecked(self=Get IncludeSourcecbox()))
   > n5 = Is Valid(InputObject=Get Parent Companion Skeletal Mesh())
   > SetEditorProperty(
    Object=Get Class Cast() -> (Class Cast:BP Live Retarget Object Reference)
    PropertyName=Original Companion Skeletal Mesh
    PropertyValue:Skeletal Mesh Object Reference=Get Parent Companion Skeletal Mesh()
   )
   > SetEditorProperty(
    Object=Get Class Cast()
    PropertyName=Original CID
    PropertyValue:Athena Character Item Definition Object Reference=Get Athena Item clear() -> (Athena Item clear:Athena Character Item Definition Object Reference)
   )
   n6 = Cast To AthenaCharacterItemDefinition(execute=@n5.Is Not Valid, Object=Get Input Parent CID() -> (Input Parent CID:Object Reference))
   > SetEditorProperty(Object=Get Class Cast(), PropertyName=Original CID, PropertyValue=@n6)
   SaveDefaultSpawnableState(execute=@n1.then_1, self=Get LevelSequenceEditorSubsystem(), ObjectBinding=Get Blueprint Proxy())
   Return Node!(execute=@n1.then_2, TargetRetargetClassBP:BP Live Retarget Object Reference=Get Class Cast())
  ]}
 Set Posseable and Bind Control Rig {nodes [
   Set Posseable and Bind Control Rig!() -> (
    then
    Input Level Sequence:Level Sequence Object Reference
    ObjectToPossess:Skeletal Mesh Component Object Reference
    Target Control Rig:Control Rig Blueprint Object Reference
    Input Local Bound Components:Map of Skeletal Mesh Component Object References to Movie Scene Binding Proxy Structures
    Rename Component:bool=~
    Component Rename:String
    only_set_possesable:bool=~
    layered rig:bool=~
   ) {is=title K2Node_FunctionEntry}
   > n1 = Sequence!(execute) -> (then_0, then_1)
   n2 = Get Local Bound Components()
   > n3 = Keys!(execute, TargetMap:Map of Skeletal Mesh Component Object References to Movie Scene Binding Proxy Structures=@n2) -> (then, Keys:Array of Skeletal Mesh Component Object References)
   n4 = Find Item(TargetArray=@n3, ItemToFind=Get ObjectToPossess() -> (ObjectToPossess:Skeletal Mesh Component Object Reference))
   n5 = "integer > integer"(A=@n4, B=-1)
   > Set Found Existing CRIG(Found Existing CRIG:bool=@n5) -> (Output_Get:bool)
   > n6 = Branch(Condition=@n5)
   > n7 = Set Target Binding(Target Binding:Movie Scene Binding Proxy Structure=Get Key Value By Index!(TargetMap:Map of Skeletal Mesh Component Object References to Movie Scene Binding Proxy Structures=@n2, Index:int=@n4) -> (Key:Skeletal Mesh Component Object Reference, Value:Movie Scene Binding Proxy Structure) {is=title K2Node_CallFunction}.Value) -> (Output_Get:Movie Scene Binding Proxy Structure)
   n8 = Get Target Control Rig() -> (Target Control Rig:Control Rig Blueprint Object Reference)
   Is Valid(exec=[@n22.then, @n7.then], InputObject=@n8)
   > n9 = Sequence!(execute) -> (then_0, then_1, then_2:Exec)
   > Branch(Condition=Get Found Existing CRIG() -> (Found Existing CRIG:bool))
   > n10 = GetControlRigs!(execute, LevelSequence:Level Sequence Object Reference=Get Input Level Sequence()) -> (then, ReturnValue:Array of Control Rig Sequencer Binding Proxy Structures)
   > n11 = For Each Loop(Array:Array of Control Rig Sequencer Binding Proxy Structures=@n10) -> (Array Element:Control Rig Sequencer Binding Proxy Structure)
   n12 = Break Control Rig Sequencer Binding Proxy!(ControlRigSequencerBindingProxy:"Control Rig Sequencer Binding Proxy Structure (by ref)"=@n11) -> (
    Proxy:Movie Scene Binding Proxy Structure
    ControlRig:Control Rig Object Reference
    Track:Movie Scene Control Rig Parameter Track Object Reference
   ) {is=title K2Node_BreakStruct}
   > Branch(Condition="Equal (MovieSceneBindingProxy)"(A:Movie Scene Binding Proxy Structure=@n12, B:Movie Scene Binding Proxy Structure=Get Target Binding()) -> (ReturnValue:bool=~) {is=title K2Node_PromotableOperator})
   n13 = Get Class(Object=@n12.ControlRig) -> (ReturnValue:Control Rig Class Reference)
   > n14 = LoadAsset(AssetPath=Split(SourceString=GetSoftClassPath!(Class:Object Class Reference=@n13) -> (ReturnValue_PathString:String), InStr=_C, SearchDir=FromEnd))
   > n15 = CastToControlRigBlueprint(Object=@n14)
   > n16 = Branch(Condition="Equal (Object)"(A=Get Target Control Rig(), B=@n15))
   > n17 = GetName(InBinding=Get Target Binding())
   > Branch(Condition="Equal Exactly (String)"(A=@n17, B=SkeletalMeshComponent0))
   > n18 = Execute Python Script!(
    execute
    PythonScript="import unreal\r\n\r\nvalue =unreal.ControlRigSequencerLibrary.is_layered_control_rig(control_rig)"
    control_rig:Control Rig Object Reference=@n12.ControlRig
   ) -> (then, ReturnValue, value:bool=~) {inputs [control_rig], outputs [value]}
   > Branch(Condition="Not Equal (Boolean)"(A=@n18.value, B="Get Layered Rig "()) -> (ReturnValue=~))
   > Execute Python Script!(
    execute
    PythonScript="import unreal\r\n\r\nunreal.ControlRigSequencerLibrary.set_control_rig_layered_mode(track, mode)"
    track:Movie Scene Control Rig Parameter Track Object Reference=@n12.Track
    mode:bool="Get Layered Rig "()
   ) -> (then, ReturnValue) {inputs [track,mode]}
   RemoveTrack!(
    execute=@n16.else
    InBinding:"Movie Scene Binding Proxy Structure (by ref)"=@n12
    TrackToRemove:Movie Scene Track Object Reference=@n12.Track
   ) -> (then)
   Set Current Control Rig(execute=@n9.then_1, Current Control Rig:Control Rig Blueprint Object Reference=@n8) -> (Output_Get:Control Rig Blueprint Object Reference)
   > "Set Layered Rig "("Layered Rig ":bool=Get layered rig() -> (layered rig:bool)) -> (Output_Get:bool)
   > Call E Bind Control Rig!(execute:Exec, self:EUW Emote Live Retarget Object Reference) -> (then:Exec) {is=title K2Node_CallDelegate}
   Branch(execute=@n9.then_2, Condition=Get Rename Component() -> (Rename Component:bool))
   > SetDisplayName!(
    execute
    InBinding:"Movie Scene Binding Proxy Structure (by ref)"=Get Target Binding()
    InDisplayName:"Text (by ref)"="To Text (String)"(InString=Get Component Rename() -> (Component Rename:String))
   ) -> (then)
   n19 = AddPossessable(execute=@n6.else, Sequence=Get Input Level Sequence(), ObjectToPossess=Get ObjectToPossess())
   > n20 = Branch(Condition=Get only_set_possesable() -> (only_set_possesable:bool))
   n21 = Set Target Binding(execute=@n20.else, Target Binding=@n19)
   > n22 = GetDisplayName!(execute:Exec, InBinding:"Movie Scene Binding Proxy Structure (by ref)"=@n21) -> (then:Exec, ReturnValue:Text)
   SaveDefaultSpawnableState(execute=@n1.then_1, self=Get LevelSequenceEditorSubsystem(), ObjectBinding=Get Blueprint Proxy())
  ]}
 Set Widget Text and Color {nodes [
   Set Widget Text and Color!() -> (then, Widget:EUW Emote Live Retarget Child Widget Object Reference, index:int) {is=title K2Node_FunctionEntry}
   > n1 = Execute Python Script!(execute, PythonScript="character = f\"Target {chr(ord('@')+(index+1))}\"", index:int=Get index() -> (index:int=0)) -> (then, ReturnValue, character:Text) {inputs [index], outputs [character]}
   > Set Text(
    Text=@n1.character
    self=Get SectionNameText_1!(self:EUW Emote Live Retarget Child Widget Object Reference=Get Widget() -> (Widget:EUW Emote Live Retarget Child Widget Object Reference)) -> (SectionNameText_1:Text Object Reference)
   )
   n2 = Get index() -> (index=0)
   > n3 = Branch(Condition="Equal (Integer)"(A=Bitwise AND(A:int=@n2, B:int=1) -> (ReturnValue:int=0) {is=title K2Node_CommutativeAssociativeBinaryOperator}, B=1))
   n4 = Get child_border_widget!(self:EUW Emote Live Retarget Child Widget Object Reference=Get Widget()) -> (child_border_widget:Border Object Reference)
   > Set BrushColor!(execute, BrushColor:Linear Color Structure=(R=1,G=1,B=1,A=0.6174), self:Border Object Reference=@n4) -> (then, Output_Get:Linear Color Structure)
   Set BrushColor(execute=@n3.else, BrushColor=(R=0.6042,G=0.6042,B=0.6042,A=0.9217), self=@n4)
  ]}
 Update CID and IKRetargets from Sequence {nodes [
   n1 = Update CID and IKRetargets from Sequence!() -> (then, Input Sequence:Movie Scene Sequence Object Reference) {is=title K2Node_FunctionEntry}
   > Set Local Sequence(Local Sequence:Movie Scene Sequence Object Reference=@n1) -> (Output_Get:Movie Scene Sequence Object Reference)
   > n2 = Sequence()
   > n3 = GetBindings(Sequence=Get Local Sequence() -> (Local Sequence:Movie Scene Sequence Object Reference))
   > n4 = For Each Loop(Array:Array of Movie Scene Binding Proxy Structures=@n3) -> (Array Element:Movie Scene Binding Proxy Structure)
   > n5 = GetBindingID(Sequence=Get Local Sequence(), InBinding=@n4)
   > n6 = For Each Loop(Array:Array of Object References=GetBoundObjects(ObjectBinding=@n5)) -> (Array Element:Object Reference)
   > n7 = Cast To SkeletalMeshComponent(Object=@n6)
   > n8 = Cast To BP_LiveRetarget(Object=GetOwner(self=@n7))
   > Set TargetRetargetClassBP(TargetRetargetClassBP=@n8)
   > Set Local IkRetarget Array(Local IkRetarget Array:Array of IKRetargeter Object References=Get IKRetargets(self=@n8)) -> (Output_Get:Array of IKRetargeter Object References)
   n9 = Get Original Companion Skeletal Mesh(self=@n8)
   > n10 = Is Valid(InputObject=@n9)
   Set Parent CID(
    execute=@n10.Is Not Valid
    Parent CID=Get Original CID!(self:BP Live Retarget Object Reference=@n8) -> (Original CID:Athena Character Item Definition Object Reference)
   )
   > n11 = For Each Loop(Array:Array of Athena Character Item Definition Object References=Get Target CIDs(self=@n8)) -> (Array Element:Athena Character Item Definition Object Reference)
   > Conditional Add Widget!(
    execute
    self
    Athena Item Array:Array of Athena Character Item Definition Object References=Get Target CIDs(self=@n8)
    Companions:bool=false
    Companions Array:Array of Skeletal Mesh Object References
   ) -> (then) {is=title K2Node_CallFunction}
   n12 = "Get (a copy)"(Array:Array of EUW Emote Live Retarget Child Widget Object References=Get Target Widgets(), Dimension 1=@n11.Array Index) -> (Output:EUW Emote Live Retarget Child Widget Object Reference)
   > Set CID(CID=@n11, self=@n12)
   > Call E Update Suffix!(execute:Exec, self:EUW Emote Live Retarget Child Widget Object Reference=@n12) -> (then:Exec) {is=title K2Node_CallDelegate}
   n13 = Execute Python Script!(
    execute=@n10.Is Valid
    PythonScript="import unreal\r\n\r\npackage_name = unreal.EditorAssetLibrary.find_asset_data(unreal.EditorAssetLibrary.get_path_name_for_loaded_asset(input_asset)).package_name\r\ndependecy = unreal.AssetRegistryDependencyOptions(include_hard_management_references =True, include_hard_package_references =True,include_searchable_names=True,include_soft_management_references =True,include_soft_package_references=True)\r\n\r\n\r\nfound_item = ''\r\nfor asset in unreal.AssetRegistryHelpers.get_asset_registry().get_referencers(package_name, dependecy):\r\n    item = unreal.load_asset(asset)\r\n    if found_item: \r\n        break\r\n    if isinstance(item, unreal.CustomCharacterPart):\r\n        target_package  = unreal.EditorAssetLibrary.find_asset_data(unreal.EditorAssetLibrary.get_path_name_for_loaded_asset(item)).package_name\r\n        for target in unreal.AssetRegistryHelpers.get_asset_registry().get_referencers(target_package, dependecy):\r\n            item = unreal.load_asset(target)\r\n            if isinstance(item, unreal.CosmeticCompanionItemDefinition):\r\n                found_item = item\r\n"
    input_asset:Skeletal Mesh Object Reference=@n9
   ) -> (then, ReturnValue, found_item:Object Reference) {inputs [input_asset], outputs [found_item]}
   > Is Valid(InputObject=@n13.found_item)
   > n14 = Cast To FortAccountItemDefinition(Object=@n13.found_item)
   > Set Parent CID(Parent CID=@n14)
   n15 = Get TargetRetargetClassBP()
   > Set Parent Companion Skeletal Mesh(Parent Companion Skeletal Mesh=Get Original Companion Skeletal Mesh(self=@n15))
   > n16 = For Each Loop(Array:Array of Skeletal Mesh Object References=Get Target Companion Skeletal Meshes(self=@n15)) -> (Array Element:Skeletal Mesh Object Reference)
   > Is Valid(InputObject=@n16)
   > n17 = Execute Python Script(
    PythonScript="import unreal\r\n\r\npackage_name = unreal.EditorAssetLibrary.find_asset_data(unreal.EditorAssetLibrary.get_path_name_for_loaded_asset(input_asset)).package_name\r\ndependecy = unreal.AssetRegistryDependencyOptions(include_hard_management_references =True, include_hard_package_references =True,include_searchable_names=True,include_soft_management_references =True,include_soft_package_references=True)\r\n\r\n\r\nfound_item = ''\r\nfor asset in unreal.AssetRegistryHelpers.get_asset_registry().get_referencers(package_name, dependecy):\r\n    item = unreal.load_asset(asset)\r\n    if found_item: \r\n        break\r\n    if isinstance(item, unreal.CustomCharacterPart):\r\n        target_package  = unreal.EditorAssetLibrary.find_asset_data(unreal.EditorAssetLibrary.get_path_name_for_loaded_asset(item)).package_name\r\n        for target in unreal.AssetRegistryHelpers.get_asset_registry().get_referencers(target_package, dependecy):\r\n            item = unreal.load_asset(target)\r\n            if isinstance(item, unreal.CosmeticCompanionItemDefinition):\r\n                found_item = item\r\n"
    input_asset=@n16
   ) {inputs [input_asset], outputs [found_item]}
   > Is Valid(InputObject=@n17.found_item)
   > n18 = Cast To FortAccountItemDefinition(Object=@n17.found_item)
   > Conditional Add Widget(Companions=true, Companions Array=Get Target Companion Skeletal Meshes(self=Get TargetRetargetClassBP()))
   n19 = "Get (a copy)"(Array=Get Target Widgets(), Dimension 1=@n16.Array Index)
   > Set CID(CID=@n18, self=@n19)
   > Set Companion Skeletal Mesh!(execute, Companion Skeletal Mesh:Skeletal Mesh Object Reference=@n16, self:EUW Emote Live Retarget Child Widget Object Reference=@n19) -> (then, Output_Get:Skeletal Mesh Object Reference)
   > Call E Update Suffix(self=@n19)
   n20 = For Each Loop(
    Exec=@n2.then_1
    Array:Array of IKRetargeter Object References=Get Local IkRetarget Array() -> (Local IkRetarget Array:Array of IKRetargeter Object References)
   ) -> (Array Element:IKRetargeter Object Reference)
   n21 = Get Target Widgets()
   > Branch(Condition="integer < integer"(A=@n20.Array Index, B=Length(TargetArray=@n21)))
   > Set IK Retargeter(IK Retargeter=@n20, self="Get (a copy)"(Array=@n21, Dimension 1=@n20.Array Index))
  ]}
 Update EUW Control Rig Widgets from Sequence {nodes [
   Update EUW Control Rig Widgets from Sequence!() -> (then, Input Level Sequence:Level Sequence Object Reference) {is=title K2Node_FunctionEntry}
   > n1 = GetControlRigs(LevelSequence=Get Input Level Sequence())
   > n2 = For Each Loop(Array:Array of Control Rig Sequencer Binding Proxy Structures=@n1) -> (Array Element:Control Rig Sequencer Binding Proxy Structure)
   n3 = Break Control Rig Sequencer Binding Proxy(ControlRigSequencerBindingProxy=@n2)
   > Set Local Proxy(Local Proxy:Movie Scene Binding Proxy Structure=@n3) -> (Output_Get:Movie Scene Binding Proxy Structure)
   > Set Local Control Rig(Local Control Rig:Control Rig Object Reference=@n3.ControlRig) -> (Output_Get:Control Rig Object Reference)
   > n4 = GetName(InBinding=Get Local Proxy() -> (Local Proxy:Movie Scene Binding Proxy Structure))
   > n5 = Branch(Condition="Equal Exactly (String)"(A=@n4, B=SkeletalMeshComponent0))
   > n6 = Execute Python Script!(
    execute
    PythonScript="import unreal\r\n\r\nvalue =unreal.ControlRigSequencerLibrary.is_layered_control_rig(control_rig)"
    control_rig:Control Rig Object Reference=Get Local Control Rig() -> (Local Control Rig:Control Rig Object Reference)
   ) -> (then, ReturnValue, value:bool=~) {inputs [control_rig], outputs [value]}
   > n7 = SetIsChecked(self=Get LayeredMasterRigCbox(), InIsChecked=@n6.value)
   n8 = GetParent(execute=[@n7.then, @n5.else], InBinding=Get Local Proxy())
   > n9 = GetName(InBinding=@n8)
   n10 = GetDisplayName!(Object:Object Reference=Get TargetRetargetClassBP()) -> (ReturnValue:String)
   > Branch(Condition="integer > integer"(A=FindSubstring(SearchIn=@n10, Substring=@n9), B=-1))
   > n11 = GetDisplayName!(execute:Exec, InBinding:"Movie Scene Binding Proxy Structure (by ref)"=Get Local Proxy()) -> (then:Exec, ReturnValue:Text)
   > Set Local Source String(Local Source String:String=ToLower!(SourceString:String="To String (Text)"(InText=@n11)) -> (ReturnValue:String)) -> (Output_Get:String)
   > n12 = Sequence!(execute) -> (then_0, then_1)
   n13 = Get Class(Object=Get Local Control Rig())
   > n14 = LoadAsset(AssetPath=Split(SourceString=GetSoftClassPath(Class=@n13), InStr=_C, SearchDir=FromEnd))
   > n15 = CastToControlRigBlueprint(Object=@n14)
   > Set Local Control Rig BP(Local Control Rig BP:Control Rig Blueprint Object Reference=@n15) -> (Output_Get:Control Rig Blueprint Object Reference)
   n16 = Branch(
    execute=@n12.then_1
    Condition="Equal Exactly (String)"(A=skeletalmeshcomponent0, B=Get Local Source String() -> (Local Source String:String))
   )
   > Set Parent Control Rig(Parent Control Rig=Get Local Control Rig BP() -> (Local Control Rig BP:Control Rig Blueprint Object Reference))
   n17 = For Each Loop(Exec=@n16.else, Array:Array of Skeletal Mesh Component Object References=Get Target Skeletal Meshes(self=Get TargetRetargetClassBP())) -> (Array Element:Skeletal Mesh Component Object Reference)
   n18 = ToLower(SourceString=GetDisplayName!(Object:Object Reference=@n17) -> (ReturnValue:String))
   n19 = Append!(A="target ", B=Split(SourceString=Get Local Source String(), InStr=__, SearchDir=FromStart)) -> (ReturnValue)
   > Branch(Condition="integer > integer"(A=FindSubstring(SearchIn=@n18, Substring=@n19)))
   > Set Control Rig(Control Rig=Get Local Control Rig BP(), self="Get (a copy)"(Array=Get Target Widgets(), Dimension 1=@n17.Array Index))
  ]}
 Update Widges from Sequence {nodes [
   Update Widges from Sequence!() -> (then, Input Level Sequence:Level Sequence Object Reference) {is=title K2Node_FunctionEntry}
   > n1 = Is Valid(InputObject=Get Input Level Sequence())
   n2 = GetCurrentLevelSequence(execute=@n1.Is Not Valid)
   > n3 = Set Local Sequence(Local Sequence:Level Sequence Object Reference=@n2) -> (Output_Get:Level Sequence Object Reference)
   n4 = Sequence!(execute=[@n3.then, @n8.then]) -> (then_0, then_1, then_2:Exec)
   > Update CID and IKRetargets from Sequence!(execute, self, Input Sequence:Movie Scene Sequence Object Reference=Get Local Sequence() -> (Local Sequence:Level Sequence Object Reference)) -> (then) {is=title K2Node_CallFunction}
   n5 = Get Animation(execute=@n4.then_1, Input Sequence=Get Local Sequence())
   > Is Valid(InputObject=@n5)
   > n6 = Cast To AnimSequence(Object=@n5)
   > Set Source Animation(Source Animation=@n6)
   Update EUW Control Rig Widgets from Sequence!(execute=@n4.then_2, self, Input Level Sequence:Level Sequence Object Reference=Get Local Sequence()) -> (then) {is=title K2Node_CallFunction}
   n7 = Get Input Level Sequence()
   LoadAsset_Blocking(execute=@n1.Is Valid, Asset=Make Soft Reference(Input=@n7))
   > n8 = Set Local Sequence(Local Sequence=@n7)
  ]}
 check companion skeletons {nodes [
   check companion skeletons() {is=title K2Node_FunctionEntry}
   > n1 = Sequence!(execute) -> (then_0, then_1)
   > n2 = GetCurrentLevelSequence()
   > Branch(Condition=IsValid!(Object:Object Reference=@n2) -> (ReturnValue))
   > Set Local Current Level Sequence(Local Current Level Sequence=@n2)
   > Clear(TargetMap:Map of Fort Account Item Definition Object References to Strings=Get Local Companion Validation() -> (Local Companion Validation:Map of Fort Account Item Definition Object References to Strings)) {is=title K2Node_CallFunction}
   > Is Valid(InputObject=Get Parent Companion Skeletal Mesh())
   > n3 = GetBindings(Sequence=Get Local Current Level Sequence())
   > n4 = For Each Loop(Array:Array of Movie Scene Binding Proxy Structures=@n3) -> (Array Element:Movie Scene Binding Proxy Structure)
   > Set Local Binding(Local Binding=@n4)
   > n5 = GetBindingID(Sequence=Get Local Current Level Sequence(), InBinding=Get Local Binding())
   > n6 = For Each Loop(Array:Array of Object References=GetBoundObjects(ObjectBinding=@n5)) -> (Array Element:Object Reference)
   > n7 = Cast To SkeletalMeshComponent(Object=@n6)
   n8 = Get SkeletalMesh(self=@n7)
   > Is Valid(InputObject=@n8)
   > Set Local Skeletal Mesh(Local Skeletal Mesh=@n8)
   > n9 = Cast To BP_LiveRetarget(Object=GetOwner(self=@n7))
   n10 = Find Item(
    TargetArray:Array of Skeletal Mesh Object References=Get Target Companion Skeletal Meshes(self=@n9)
    ItemToFind:"Skeletal Mesh Object Reference (by ref)"=Get Local Skeletal Mesh()
   )
   > Branch(Condition="Not Equal (Integer)"(A=@n10, B=-1))
   > n11 = Set local widget data(local widget data:EUW Emote Live Retarget Child Widget Object Reference="Get (a copy)"(Array=Get Target Widgets(), Dimension 1=@n10)) -> (Output_Get:EUW Emote Live Retarget Child Widget Object Reference)
   > Is Valid(InputObject=@n11)
   n12 = Get local widget data() -> (local widget data:EUW Emote Live Retarget Child Widget Object Reference)
   > Branch(Condition=IsChecked(self=Get IncludeTargetCbox(self=@n12)))
   > n13 = Execute Python Script!(
    execute
    PythonScript="import unreal\r\nfound = ''\r\n\r\ndependecy = unreal.AssetRegistryDependencyOptions(include_hard_management_references =True, include_hard_package_references =True,include_searchable_names=True,include_soft_management_references =True,include_soft_package_references=True)\r\npackage_name = unreal.EditorAssetLibrary.find_asset_data(unreal.EditorAssetLibrary.get_path_name_for_loaded_asset(asset)).package_name\r\n\r\nfor item in unreal.AssetRegistryHelpers.get_asset_registry().get_referencers(package_name, dependecy):\r\n   ## print(item)\r\n    if found:\r\n       break\r\n    loaded_item = unreal.load_asset(item)\r\n    if isinstance(loaded_item, unreal.Blueprint):\r\n        print('checking...', loaded_item.get_name())\r\n        package = unreal.EditorAssetLibrary.find_asset_data(unreal.EditorAssetLibrary.get_path_name_for_loaded_asset(loaded_item)).package_name\r\n        ## print(unreal.AssetRegistryHelpers.get_asset_registry().get_referencers(package, dependecy))\r\n        for connections in unreal.AssetRegistryHelpers.get_asset_registry().get_referencers(package, dependecy) or []:\r\n            connected_item = unreal.load_asset(connections)\r\n            if isinstance(connected_item, unreal.Object) and connected_item.get_name().startswith('AC'):\r\n                found = connected_item\r\n                break"
    asset:Skeletal Mesh Object Reference=Get Companion Skeletal Mesh(self=Get local widget data())
   ) -> (then, ReturnValue, found:Animatable Creature Object Reference) {inputs [asset], outputs [found]}
   > Set Local New AC file(Local New AC file:Animatable Creature Object Reference=@n13.found) -> (Output_Get:Animatable Creature Object Reference)
   > n14 = Execute Python Script(
    PythonScript="import unreal\r\nfound = ''\r\n\r\ndependecy = unreal.AssetRegistryDependencyOptions(include_hard_management_references =True, include_hard_package_references =True,include_searchable_names=True,include_soft_management_references =True,include_soft_package_references=True)\r\npackage_name = unreal.EditorAssetLibrary.find_asset_data(unreal.EditorAssetLibrary.get_path_name_for_loaded_asset(asset)).package_name\r\n\r\nfor item in unreal.AssetRegistryHelpers.get_asset_registry().get_referencers(package_name, dependecy):\r\n   ## print(item)\r\n    if found:\r\n       break\r\n    loaded_item = unreal.load_asset(item)\r\n    if isinstance(loaded_item, unreal.Blueprint):\r\n        print('checking...', loaded_item.get_name())\r\n        package = unreal.EditorAssetLibrary.find_asset_data(unreal.EditorAssetLibrary.get_path_name_for_loaded_asset(loaded_item)).package_name\r\n        ## print(unreal.AssetRegistryHelpers.get_asset_registry().get_referencers(package, dependecy))\r\n        for connections in unreal.AssetRegistryHelpers.get_asset_registry().get_referencers(package, dependecy) or []:\r\n            connected_item = unreal.load_asset(connections)\r\n            if isinstance(connected_item, unreal.Object) and connected_item.get_name().startswith('AC'):\r\n                found = connected_item\r\n                break"
    asset="Get (a copy)"(Array:Array of Skeletal Mesh Object References=Get Target Companion Skeletal Meshes(self=@n9), Dimension 1=@n10) -> (Output:Skeletal Mesh Object Reference)
   ) {inputs [asset], outputs [found]}
   > Set Local AC file(Local AC file:Animatable Creature Object Reference=@n14.found) -> (Output_Get:Animatable Creature Object Reference)
   n15 = "Not Equal (Enum)"!(
    A:ENUM_CompanionArchetype Enum=Get Archetype!(self:Animatable Creature Object Reference=Get Local AC file() -> (Local AC file:Animatable Creature Object Reference)) -> (Archetype:ENUM_CompanionArchetype Enum=NewEnumerator0)
    B:ENUM_CompanionArchetype Enum=Get Archetype(self=Get Local New AC file() -> (Local New AC file:Animatable Creature Object Reference)) -> (Archetype=NewEnumerator0)
   ) -> (ReturnValue:bool)
   n16 = Get local widget data()
   n17 = "Not Equal (Object)"(
    A:Object Reference=Get Skeleton(self=Get Companion Skeletal Mesh(self=@n16))
    B:Object Reference=Get Skeleton(self=Get Local Skeletal Mesh())
   ) -> (ReturnValue:bool=~) {is=title K2Node_PromotableOperator}
   > Branch(Condition=OR Boolean(A=@n15, B=@n17))
   n18 = NOT Boolean(A=@n15)
   n19 = NOT Boolean(A=@n17)
   > Add(
    TargetMap:Map of Fort Account Item Definition Object References to Strings=Get Local Companion Validation()
    Key:"Fort Account Item Definition Object Reference (by ref)"=Get CID(self=Get local widget data())
    Value:"String (by ref)"=Append!(
     A="ArchType Match:  "
     B="To String (Boolean)"(InBool=@n18)
     C:String="  |   "
     D:String="Skeleton Match: "
     E:String="To String (Boolean)"(InBool=@n19)
    ) -> (ReturnValue)
   ) {is=title K2Node_CallFunction}
   n20 = Branch(
    execute=@n1.then_1
    Condition=Is Not Empty!(TargetMap:Map of Fort Account Item Definition Object References to Strings=Get Local Companion Validation()) -> (ReturnValue:bool) {is=title K2Node_CallFunction}
   )
   > n21 = Sequence!(execute) -> (then_0, then_1, then_2:Exec)
   > n22 = "For Each Loop (Map)"!(execute:Exec, MapPin:Map of Fort Account Item Definition Object References to Strings=Get Local Companion Validation(), BreakPin:Exec) -> (then:Exec, KeyPin:Fort Account Item Definition Object Reference, ValuePin:String, CompletedPin:Exec)
   > Set local combined check string(local combined check string:String=Append!(
     A=Get local combined check string() -> (local combined check string:String)
     B="\r\n"
     C=GetObjectName(Object=@n22)
     D="  :  "
     E=@n22.ValuePin
     F:String="\r\n"
    ) -> (ReturnValue)) -> (Output_Get:String)
   n23 = Set local warning message(
    execute=@n21.then_1
    local warning message:String=Append!(
     A=Get local warning message() -> (local warning message:String)
     B="\r\n"
     C=Get local combined check string()
     D="\r\n"
     E="These Items will need their skeletons changed in their Linked Anims using right click -> replace skeleton."
    ) -> (ReturnValue)
   ) -> (Output_Get:String)
   > n24 = Show Message Dialog(
    Title="To Text (String)"(InString=Skeletal Changes)
    Message="To Text (String)"(InString=@n23)
    MessageType=YesNoCancel
    DefaultValue=No
    MessageCategory=Warning
   )
   > Set Local Message Warning Return(Local Message Warning Return:EAppReturnType Enum=@n24) -> (Output_Get:EAppReturnType Enum)
   Return Node!(
    execute=[@n21.then_2, @n20.else]
    Continue:EAppReturnType Enum=Get Local Message Warning Return() -> (Local Message Warning Return:EAppReturnType Enum)
   )
  ]}
]
variables [
 Anim Browser {type=Animation_Browser_C, default=null}
 Anim Sequence {type=string, default=""}
 Animation Path {type=string, default=""}
 binding {type=MovieSceneBindingProxy, default {bindingId=#4, sequence=null}}
 Blueprint Proxy {type=MovieSceneBindingProxy, default {bindingId=#4, sequence=null}}
 bound ID {type=Object, default=null}
 Children Companion Skeletal Mesh {type="array<SkeletalMesh>", default []}
 Companions Project {type=boolean}
 Current Control Rig {type=ControlRigBlueprint, default=null, class=~}
 Current Level Sequence {type=LevelSequence, default=null}
 Default Human CIDs {
  type="array<AthenaCharacterItemDefinition>"
  default [
   /BRCosmetics/Athena/Items/Cosmetics/Characters/CID_A_100_Athena_Commando_M_Downpour_KC39P
   /BRCosmetics/Athena/Items/Cosmetics/Characters/Character_BasilStrong
   /BRCosmetics/Athena/Items/Cosmetics/Characters/CID_231_Athena_Commando_F_RedRiding
   /BRCosmetics/Athena/Items/Cosmetics/Characters/CID_719_Athena_Commando_F_Blonde
  ]
 }
 E Bind Control Rig {default=()}
 E Create New Animation {default=()}
 E Export Animations {default=()}
 E Main Process {default=()}
 E Remove Widget {default=()}
 EUW Anim Dialog {type=EUW_Animation_Dialog_C, default=null}
 "Layered Rig " {type=boolean}
 Link New Paths {default=()}
 Local Bound Components {default {}}
 max_targets {type=integer, default=10}
 New Actor Skeletal Mesh Component {type=SkeletalMeshComponent}
 Parent CID {type=FortAccountItemDefinition, default=null}
 Parent Companion Skeletal Mesh {type=SkeletalMesh, default=null}
 Parent Control Rig {type=ControlRigBlueprint, default=null, class=~}
 Project {type=string, default=AnimationBrowser_BREmotes_Project}
 replace {type=boolean}
 Set Targets {type=integer, default=4}
 Source Animation {type=AnimSequence, default=null}
 Target Binding {type=MovieSceneBindingProxy, default {bindingId=#4, sequence=null}}
 Target CIDs {type="array<AthenaCharacterItemDefinition>", default []}
 Target IKRetargets {type="array<IKRetargeter>", default []}
 Target Skeletal Mesh Components {type="array<SkeletalMeshComponent>", default []}
 Target Widgets {type="array<EUW_EmoteLiveRetarget_Child_Widget_C>", default []}
 TargetRetargetClassBP {type=BP_LiveRetarget_C, default=null}
]
widgets {
 info {rootWidgetClass=&CanvasPanel, widgetCount=79}
 widgets [
  CanvasPanel_33 {widgetClassPath=&CanvasPanel, widget=CanvasPanel_33}
  ScaleBox_0 {
   widgetClassPath=&ScaleBox
   parent=CanvasPanel_33
   widget=ScaleBox_0
   slot {class=CanvasPanelSlot, properties {layoutData {alignment=<0 0>, anchors {maximum=<1 1>, minimum=<0 0>}, offsets=<ltrb 0 0 0 0>}}}
  }
  SizeBox_454 {
   widgetClassPath=&SizeBox
   parent=ScaleBox_0
   widget=SizeBox_454
   properties {bOverride_MaxDesiredHeight=true, bOverride_MaxDesiredWidth=true, maxDesiredHeight=1000, maxDesiredWidth=200}
   slot {class=ScaleBoxSlot, properties {horizontalAlignment=HAlign_Fill, verticalAlignment=VAlign_Fill}}
  }
  Border_0 {
   widgetClassPath=&Border
   parent=SizeBox_454
   widget=Border_0
   properties {brushColor=<rgba 0.0191 0.0191 0.0191 1>, padding=<ltrb 0 0 0 0>}
   slot {class=SizeBoxSlot}
  }
  VerticalBox_0 {widgetClassPath=&VerticalBox, parent=Border_0, widget=VerticalBox_0, slot {class=BorderSlot, properties {padding=<ltrb 0 0 0 0>}}}
  SizeBox_0 {
   widgetClassPath=&SizeBox
   parent=VerticalBox_0
   widget=SizeBox_0
   properties {bOverride_HeightOverride=true, heightOverride=400}
   slot {class=VerticalBoxSlot}
  }
  AnimBrowserWidget {widgetClassPath=&NamedSlot, bIsVariable=true, parent=SizeBox_0, widget=AnimBrowserWidget, slot {class=SizeBoxSlot}}
  line_1 {
   widgetClassPath=&CommonBorder
   parent=VerticalBox_0
   widget=line_1
   properties {background {
     bIsDynamicallyLoaded=false
     drawAs=RoundedBox
     imageType=NoImage
     mirroring=NoMirror
     resourceName=None
     resourceObject=null
     tiling=NoTile
     imageSize=<18 18>
     margin=<ltrb 0.4 0.4 0.4 0.4>
     outlineSettings {
      bUseBrushTransparency=false
      roundingType=HalfHeightRadius
      width=1.4037
      color {colorUseRule=UseColor_Specified, specifiedColor=<rgba 0 0 0 0>}
      cornerRadii=<-0.4977 0.3303 0 -0.5778>
     }
     tintColor {colorUseRule=UseColor_Specified, specifiedColor=<rgba 0 0 0 1>}
     uVRegion {bIsValid=false, max=<0 0>, min=<0 0>}
    }}
   slot {class=VerticalBoxSlot}
  }
  ScaleBox_72 {
   widgetClassPath=&ScaleBox
   parent=VerticalBox_0
   widget=ScaleBox_72
   slot {class=VerticalBoxSlot, properties {horizontalAlignment=HAlign_Center}}
  }
  HorizontalBox_109 {
   widgetClassPath=&HorizontalBox
   parent=VerticalBox_0
   widget=HorizontalBox_109
   slot {class=VerticalBoxSlot, properties {horizontalAlignment=HAlign_Center, verticalAlignment=VAlign_Center, padding=<ltrb 15 0 0 0>}}
  }
  VerticalBox_54 {widgetClassPath=&VerticalBox, parent=HorizontalBox_109, widget=VerticalBox_54, slot {class=HorizontalBoxSlot}}
  HorizontalBox_172 {
   widgetClassPath=&HorizontalBox
   parent=VerticalBox_54
   widget=HorizontalBox_172
   slot {class=VerticalBoxSlot, properties {horizontalAlignment=HAlign_Center, verticalAlignment=VAlign_Center}}
  }
  AddWidgetbutton {
   widgetClassPath=&EditorUtilityButton
   parent=HorizontalBox_172
   widget=AddWidgetbutton
   slot {class=HorizontalBoxSlot, properties {horizontalAlignment=HAlign_Left}}
  }
  HorizontalBox_74 {
   widgetClassPath=&HorizontalBox
   parent=AddWidgetbutton
   widget=HorizontalBox_74
   slot {class=ButtonSlot, properties {horizontalAlignment=HAlign_Fill, verticalAlignment=VAlign_Fill, padding=<ltrb 0 0 0 0>}}
  }
  AddwidgetIcon {
   widgetClassPath=&Image
   bIsVariable=true
   parent=HorizontalBox_74
   widget=AddwidgetIcon
   properties {
    visibility=HitTestInvisible
    brush {
     bIsDynamicallyLoaded=false
     drawAs=Image
     imageType=FullColor
     mirroring=NoMirror
     resourceName=None
     resourceObject=/AnimInEdTools/AnimationBrowser/Icons/Icon_LevelSequence_add
     tiling=NoTile
     imageSize=<14 14>
     margin=<ltrb 0 0 0 0>
     outlineSettings {
      bUseBrushTransparency=false
      roundingType=HalfHeightRadius
      width=0
      color {colorUseRule=UseColor_Specified, specifiedColor=<rgba 0 0 0 0>}
      cornerRadii=<0 0 0 0>
     }
     tintColor {colorUseRule=UseColor_Specified, specifiedColor=<rgba 0.1823 0.638 1 1>}
     uVRegion {bIsValid=false, max=<0 0>, min=<0 0>}
    }
   }
   slot {class=HorizontalBoxSlot, properties {horizontalAlignment=HAlign_Center, verticalAlignment=VAlign_Center, padding=<ltrb 4 2 4 2>}}
  }
  Addwidgettext {
   widgetClassPath=&TextBlock
   parent=HorizontalBox_74
   widget=Addwidgettext
   properties {
    text=Add Target
    visibility=SelfHitTestInvisible
    colorAndOpacity {colorUseRule=UseColor_Specified, specifiedColor=<rgba 0.5255 0.5255 0.5255 1>}
    font {
     bForceMonospaced=false
     bMaterialIsStencil=false
     fontMaterial=null
     fontObject=&Roboto
     letterSpacing=0
     monospacedWidth=1
     size=9
     skewAmount=0
     typefaceFontName=Bold
     outlineSettings {
      bApplyOutlineToDropShadows=false
      bMiteredCorners=false
      bSeparateFillAlpha=false
      outlineMaterial=null
      outlineSize=0
      outlineColor=<rgba 0 0 0 1>
     }
    }
   }
   slot {class=HorizontalBoxSlot, properties {verticalAlignment=VAlign_Center}}
  }
  SubWidgetbutton {
   widgetClassPath=&EditorUtilityButton
   parent=HorizontalBox_172
   widget=SubWidgetbutton
   slot {class=HorizontalBoxSlot, properties {horizontalAlignment=HAlign_Left}}
  }
  HorizontalBox {
   widgetClassPath=&HorizontalBox
   parent=SubWidgetbutton
   widget=HorizontalBox
   slot {class=ButtonSlot, properties {horizontalAlignment=HAlign_Fill, verticalAlignment=VAlign_Fill, padding=<ltrb 0 0 0 0>}}
  }
  AddwidgetIcon_1 {
   widgetClassPath=&Image
   bIsVariable=true
   parent=HorizontalBox
   widget=AddwidgetIcon_1
   properties {
    visibility=HitTestInvisible
    brush {
     bIsDynamicallyLoaded=false
     drawAs=Image
     imageType=FullColor
     mirroring=NoMirror
     resourceName=None
     resourceObject=/FortUILibrary/Foundation/Buttons/Materials/Arrows/M_Button_Minus_Base
     tiling=NoTile
     imageSize=<17 17>
     margin=<ltrb 0 0 0 0>
     outlineSettings {
      bUseBrushTransparency=false
      roundingType=HalfHeightRadius
      width=0
      color {colorUseRule=UseColor_Specified, specifiedColor=<rgba 0 0 0 0>}
      cornerRadii=<0 0 0 0>
     }
     tintColor {colorUseRule=UseColor_Specified, specifiedColor=<rgba 0.4948 0.0166 0.0438 1>}
     uVRegion {bIsValid=false, max=<0 0>, min=<0 0>}
    }
   }
   slot {class=HorizontalBoxSlot, properties {horizontalAlignment=HAlign_Center, verticalAlignment=VAlign_Center, padding=<ltrb 4 2 4 2>}}
  }
  RemoveWidgetText {
   widgetClassPath=&TextBlock
   parent=HorizontalBox
   widget=RemoveWidgetText
   properties {
    text=Remove Last Target
    visibility=SelfHitTestInvisible
    colorAndOpacity {colorUseRule=UseColor_Specified, specifiedColor=<rgba 0.5255 0.5255 0.5255 1>}
    font {
     bForceMonospaced=false
     bMaterialIsStencil=false
     fontMaterial=null
     fontObject=&Roboto
     letterSpacing=0
     monospacedWidth=1
     size=9
     skewAmount=0
     typefaceFontName=Bold
     outlineSettings {
      bApplyOutlineToDropShadows=false
      bMiteredCorners=false
      bSeparateFillAlpha=false
      outlineMaterial=null
      outlineSize=0
      outlineColor=<rgba 0 0 0 1>
     }
    }
   }
   slot {class=HorizontalBoxSlot, properties {verticalAlignment=VAlign_Center}}
  }
  ClearButton {widgetClassPath=&EditorUtilityButton, parent=HorizontalBox_172, widget=ClearButton, slot {class=HorizontalBoxSlot}}
  HorizontalBox_202 {
   widgetClassPath=&HorizontalBox
   parent=ClearButton
   widget=HorizontalBox_202
   slot {class=ButtonSlot, properties {horizontalAlignment=HAlign_Fill, verticalAlignment=VAlign_Fill, padding=<ltrb 0 0 0 0>}}
  }
  Image_259 {
   widgetClassPath=&Image
   bIsVariable=true
   parent=HorizontalBox_202
   widget=Image_259
   properties {
    visibility=HitTestInvisible
    brush {
     bIsDynamicallyLoaded=false
     drawAs=Image
     imageType=FullColor
     mirroring=NoMirror
     resourceName=None
     resourceObject=/AnimInEdTools/AnimationPicker/Icons/AP_icon_delete
     tiling=NoTile
     imageSize=<16 16>
     margin=<ltrb 0 0 0 0>
     outlineSettings {
      bUseBrushTransparency=false
      roundingType=HalfHeightRadius
      width=0
      color {colorUseRule=UseColor_Specified, specifiedColor=<rgba 0 0 0 0>}
      cornerRadii=<0 0 0 0>
     }
     tintColor {colorUseRule=UseColor_Specified, specifiedColor=<rgba 1 1 1 1>}
     uVRegion {bIsValid=false, max=<0 0>, min=<0 0>}
    }
   }
   slot {class=HorizontalBoxSlot, properties {horizontalAlignment=HAlign_Center, verticalAlignment=VAlign_Center}}
  }
  ClearText {
   widgetClassPath=&TextBlock
   parent=HorizontalBox_202
   widget=ClearText
   properties {
    text=Clear All
    visibility=SelfHitTestInvisible
    colorAndOpacity {colorUseRule=UseColor_Specified, specifiedColor=<rgba 0.5255 0.5255 0.5255 1>}
    font {
     bForceMonospaced=false
     bMaterialIsStencil=false
     fontMaterial=null
     fontObject=&Roboto
     letterSpacing=0
     monospacedWidth=1
     size=9
     skewAmount=0
     typefaceFontName=Bold
     outlineSettings {
      bApplyOutlineToDropShadows=false
      bMiteredCorners=false
      bSeparateFillAlpha=false
      outlineMaterial=null
      outlineSize=0
      outlineColor=<rgba 0 0 0 1>
     }
    }
   }
   slot {class=HorizontalBoxSlot, properties {horizontalAlignment=HAlign_Center, verticalAlignment=VAlign_Center, padding=<ltrb 4 2 4 2>}}
  }
  UpdateDetailsButton {widgetClassPath=&EditorUtilityButton, parent=HorizontalBox_172, widget=UpdateDetailsButton, slot {class=HorizontalBoxSlot}}
  TextBlock {
   widgetClassPath=&TextBlock
   parent=UpdateDetailsButton
   widget=TextBlock
   properties {
    text=Update UI from Current Sequence
    visibility=SelfHitTestInvisible
    colorAndOpacity {colorUseRule=UseColor_Specified, specifiedColor=<rgba 0.5255 0.5255 0.5255 1>}
    font {
     bForceMonospaced=false
     bMaterialIsStencil=false
     fontMaterial=null
     fontObject=&Roboto
     letterSpacing=0
     monospacedWidth=1
     size=9
     skewAmount=0
     typefaceFontName=Bold
     outlineSettings {
      bApplyOutlineToDropShadows=false
      bMiteredCorners=false
      bSeparateFillAlpha=false
      outlineMaterial=null
      outlineSize=0
      outlineColor=<rgba 0 0 0 1>
     }
    }
   }
   slot {class=ButtonSlot}
  }
  ?button {
   widgetClassPath=&EditorUtilityButton
   parent=HorizontalBox_172
   widget=?button
   properties {toolTipText=Open Confluence Page}
   slot {class=HorizontalBoxSlot}
  }
  TextBlock_14 {
   widgetClassPath=&TextBlock
   parent=?button
   widget=TextBlock_14
   properties {
    text=?
    visibility=HitTestInvisible
    colorAndOpacity {colorUseRule=UseColor_Specified, specifiedColor=<rgba 0 1 0 1>}
    font {
     bForceMonospaced=false
     bMaterialIsStencil=false
     fontMaterial=null
     fontObject=&Roboto
     letterSpacing=0
     monospacedWidth=1
     size=12
     skewAmount=0
     typefaceFontName=Bold
     outlineSettings {
      bApplyOutlineToDropShadows=false
      bMiteredCorners=false
      bSeparateFillAlpha=false
      outlineMaterial=null
      outlineSize=0
      outlineColor=<rgba 0 0 0 1>
     }
    }
   }
   slot {class=ButtonSlot, properties {horizontalAlignment=HAlign_Left, padding=<ltrb 0 0 0 0>}}
  }
  HorizontalBox_80 {
   widgetClassPath=&HorizontalBox
   parent=VerticalBox_54
   widget=HorizontalBox_80
   slot {class=VerticalBoxSlot, properties {horizontalAlignment=HAlign_Center, verticalAlignment=VAlign_Center, padding=<ltrb 0 2 0 0>}}
  }
  NewAnimationButton {
   widgetClassPath=&EditorUtilityButton
   parent=HorizontalBox_80
   widget=NewAnimationButton
   slot {class=HorizontalBoxSlot, properties {horizontalAlignment=HAlign_Left}}
  }
  HorizontalBox_71 {
   widgetClassPath=&HorizontalBox
   parent=NewAnimationButton
   widget=HorizontalBox_71
   slot {class=ButtonSlot, properties {horizontalAlignment=HAlign_Fill, verticalAlignment=VAlign_Fill}}
  }
  Image_46 {
   widgetClassPath=&Image
   bIsVariable=true
   parent=HorizontalBox_71
   widget=Image_46
   properties {
    visibility=HitTestInvisible
    brush {
     bIsDynamicallyLoaded=false
     drawAs=Image
     imageType=FullColor
     mirroring=NoMirror
     resourceName=None
     resourceObject=&Icon_LevelSequence_add
     tiling=NoTile
     imageSize=<14 14>
     margin=<ltrb 0 0 0 0>
     outlineSettings {
      bUseBrushTransparency=false
      roundingType=HalfHeightRadius
      width=0
      color {colorUseRule=UseColor_Specified, specifiedColor=<rgba 0 0 0 0>}
      cornerRadii=<0 0 0 0>
     }
     tintColor {colorUseRule=UseColor_Specified, specifiedColor=<rgba 0.2108 1 0 1>}
     uVRegion {bIsValid=false, max=<0 0>, min=<0 0>}
    }
   }
   slot {class=HorizontalBoxSlot, properties {horizontalAlignment=HAlign_Center, verticalAlignment=VAlign_Center, padding=<ltrb 0 0 2 0>}}
  }
  NewAnimText {
   widgetClassPath=&TextBlock
   parent=HorizontalBox_71
   widget=NewAnimText
   properties {
    text=New Emote Animation
    visibility=HitTestInvisible
    colorAndOpacity {colorUseRule=UseColor_Specified, specifiedColor=<rgba 0.5255 0.5255 0.5255 1>}
    font {
     bForceMonospaced=false
     bMaterialIsStencil=false
     fontMaterial=null
     fontObject=&Roboto
     letterSpacing=0
     monospacedWidth=1
     size=9
     skewAmount=0
     typefaceFontName=Bold
     outlineSettings {
      bApplyOutlineToDropShadows=false
      bMiteredCorners=false
      bSeparateFillAlpha=false
      outlineMaterial=null
      outlineSize=0
      outlineColor=<rgba 0 0 0 1>
     }
    }
   }
   slot {class=HorizontalBoxSlot, properties {verticalAlignment=VAlign_Center}}
  }
  ReplaceSelectedActortoEmoteRetargetButton {
   widgetClassPath=&EditorUtilityButton
   parent=HorizontalBox_80
   widget=ReplaceSelectedActortoEmoteRetargetButton
   slot {class=HorizontalBoxSlot}
  }
  ReplaceSelectedActortoEmoteRetarget {
   widgetClassPath=&TextBlock
   parent=ReplaceSelectedActortoEmoteRetargetButton
   widget=ReplaceSelectedActortoEmoteRetarget
   properties {
    text=Replace Selected Actor with Emote Retarget BP
    visibility=SelfHitTestInvisible
    colorAndOpacity {colorUseRule=UseColor_Specified, specifiedColor=<rgba 0.5255 0.5255 0.5255 1>}
    font {
     bForceMonospaced=false
     bMaterialIsStencil=false
     fontMaterial=null
     fontObject=&Roboto
     letterSpacing=0
     monospacedWidth=1
     size=9
     skewAmount=0
     typefaceFontName=Bold
     outlineSettings {
      bApplyOutlineToDropShadows=false
      bMiteredCorners=false
      bSeparateFillAlpha=false
      outlineMaterial=null
      outlineSize=0
      outlineColor=<rgba 0 0 0 1>
     }
    }
   }
   slot {class=ButtonSlot, properties {horizontalAlignment=HAlign_Left, padding=<ltrb 1 2 4 2>}}
  }
  CreateNewSequenceButton {
   widgetClassPath=&EditorUtilityButton
   parent=HorizontalBox_80
   widget=CreateNewSequenceButton
   slot {class=HorizontalBoxSlot, properties {verticalAlignment=VAlign_Center}}
  }
  TextBlock_0 {
   widgetClassPath=&TextBlock
   parent=CreateNewSequenceButton
   widget=TextBlock_0
   properties {
    text=Add/Update Assets in Current Level Sequence
    visibility=SelfHitTestInvisible
    colorAndOpacity {colorUseRule=UseColor_Specified, specifiedColor=<rgba 0.5255 0.5255 0.5255 1>}
    font {
     bForceMonospaced=false
     bMaterialIsStencil=false
     fontMaterial=null
     fontObject=&Roboto
     letterSpacing=0
     monospacedWidth=1
     size=9
     skewAmount=0
     typefaceFontName=Bold
     outlineSettings {
      bApplyOutlineToDropShadows=false
      bMiteredCorners=false
      bSeparateFillAlpha=false
      outlineMaterial=null
      outlineSize=0
      outlineColor=<rgba 0 0 0 1>
     }
    }
    renderTransform {angle=0, scale=<1 1.1>, shear=<0 0>, translation=<5 0>}
   }
   slot {class=ButtonSlot, properties {padding=<ltrb 1 2 4 2>}}
  }
  CommonBorder_0 {
   widgetClassPath=&CommonBorder
   parent=HorizontalBox_109
   widget=CommonBorder_0
   properties {
    background {
     bIsDynamicallyLoaded=false
     drawAs=Border
     imageType=NoImage
     mirroring=NoMirror
     resourceName=None
     resourceObject=null
     tiling=NoTile
     imageSize=<32 32>
     margin=<ltrb 1 1 1 1>
     outlineSettings {
      bUseBrushTransparency=false
      roundingType=HalfHeightRadius
      width=0
      color {colorUseRule=UseColor_Specified, specifiedColor=<rgba 0 0 0 0>}
      cornerRadii=<0 0 0 0>
     }
     tintColor {colorUseRule=UseColor_Specified, specifiedColor=<rgba 0 0 0 1>}
     uVRegion {bIsValid=false, max=<0 0>, min=<0 0>}
    }
    padding=<ltrb 1 1 1 1>
   }
   slot {class=HorizontalBoxSlot, properties {horizontalAlignment=HAlign_Center, verticalAlignment=VAlign_Center, padding=<ltrb 5 1 0 1>}}
  }
  CommonBorder_84 {
   widgetClassPath=&CommonBorder
   parent=CommonBorder_0
   widget=CommonBorder_84
   properties {
    background {
     bIsDynamicallyLoaded=false
     drawAs=Box
     imageType=NoImage
     mirroring=NoMirror
     resourceName=None
     resourceObject=null
     tiling=NoTile
     imageSize=<32 32>
     margin=<ltrb 1 1 1 1>
     outlineSettings {
      bUseBrushTransparency=false
      roundingType=HalfHeightRadius
      width=0
      color {colorUseRule=UseColor_Specified, specifiedColor=<rgba 0 0 0 0>}
      cornerRadii=<0 0 0 0>
     }
     tintColor {colorUseRule=UseColor_Specified, specifiedColor=<rgba 0.0275 0.0348 0.0346 0.3826>}
     uVRegion {bIsValid=false, max=<0 0>, min=<0 0>}
    }
    padding=<ltrb 0 0 0 0>
   }
   slot {class=BorderSlot, properties {padding=<ltrb 1 1 1 1>}}
  }
  VerticalBox_78 {
   widgetClassPath=&VerticalBox
   parent=CommonBorder_84
   widget=VerticalBox_78
   slot {class=BorderSlot, properties {padding=<ltrb 0 0 0 0>}}
  }
  TextBlock_63 {
   widgetClassPath=&TextBlock
   parent=VerticalBox_78
   widget=TextBlock_63
   properties {
    text=Options:
    font {
     bForceMonospaced=false
     bMaterialIsStencil=false
     fontMaterial=null
     fontObject=&Roboto
     letterSpacing=0
     monospacedWidth=1
     size=13.5
     skewAmount=0
     typefaceFontName=Bold
     outlineSettings {
      bApplyOutlineToDropShadows=false
      bMiteredCorners=false
      bSeparateFillAlpha=false
      outlineMaterial=null
      outlineSize=0
      outlineColor=<rgba 0 0 0 1>
     }
    }
   }
   slot {class=VerticalBoxSlot, properties {horizontalAlignment=HAlign_Center, verticalAlignment=VAlign_Center}}
  }
  HorizontalBox_82 {
   widgetClassPath=&HorizontalBox
   parent=VerticalBox_78
   widget=HorizontalBox_82
   slot {class=VerticalBoxSlot, properties {padding=<ltrb 4 0 3 0>}}
  }
  TextBlock_56 {
   widgetClassPath=&TextBlock
   parent=HorizontalBox_82
   widget=TextBlock_56
   properties {
    text=Check out Level Sequence:
    font {
     bForceMonospaced=false
     bMaterialIsStencil=false
     fontMaterial=null
     fontObject=null
     letterSpacing=0
     monospacedWidth=1
     size=10.5
     skewAmount=0
     typefaceFontName=Default
     outlineSettings {
      bApplyOutlineToDropShadows=false
      bMiteredCorners=false
      bSeparateFillAlpha=false
      outlineMaterial=null
      outlineSize=0
      outlineColor=<rgba 0 0 0 1>
     }
    }
    renderTransform {angle=0, scale=<1 1.1>, shear=<0 0>, translation=<0 0>}
   }
   slot {class=HorizontalBoxSlot, properties {verticalAlignment=VAlign_Center, padding=<ltrb 1 0 0 0>}}
  }
  checkoutLS_cbox {
   widgetClassPath=&EditorUtilityCheckBox
   bIsVariable=true
   parent=HorizontalBox_82
   widget=checkoutLS_cbox
   properties {checkedState=Checked, toolTipText=Check out the level sequence and save when updating}
   slot {class=HorizontalBoxSlot, properties {padding=<ltrb 1 0 0 0>}}
  }
  line_8 {
   widgetClassPath=&CommonBorder
   parent=HorizontalBox_82
   widget=line_8
   properties {
    background {
     bIsDynamicallyLoaded=false
     drawAs=Box
     imageType=NoImage
     mirroring=NoMirror
     resourceName=None
     resourceObject=null
     tiling=NoTile
     imageSize=<18 18>
     margin=<ltrb 0.4 0.4 0.4 0.4>
     outlineSettings {
      bUseBrushTransparency=false
      roundingType=HalfHeightRadius
      width=-9.2626
      color {colorUseRule=UseColor_Specified, specifiedColor=<rgba 0 0 0 0>}
      cornerRadii=<0 0 0 0>
     }
     tintColor {colorUseRule=UseColor_Specified, specifiedColor=<rgba 0 0 0 1>}
     uVRegion {bIsValid=false, max=<0 0>, min=<0 0>}
    }
    padding=<ltrb 1.8 2 1.8 2>
    renderTransform {angle=0, scale=<0.6 1>, shear=<0 0>, translation=<0 0>}
   }
   slot {class=HorizontalBoxSlot, properties {padding=<ltrb 0 0 1 0>}}
  }
  TextBlock_66 {
   widgetClassPath=&TextBlock
   parent=HorizontalBox_82
   widget=TextBlock_66
   properties {
    text="Generate  or Update Linked Anims"
    font {
     bForceMonospaced=false
     bMaterialIsStencil=false
     fontMaterial=null
     fontObject=null
     letterSpacing=0
     monospacedWidth=1
     size=10.5
     skewAmount=0
     typefaceFontName=Default
     outlineSettings {
      bApplyOutlineToDropShadows=false
      bMiteredCorners=false
      bSeparateFillAlpha=false
      outlineMaterial=null
      outlineSize=0
      outlineColor=<rgba 0 0 0 1>
     }
    }
   }
   slot {class=HorizontalBoxSlot, properties {verticalAlignment=VAlign_Center, padding=<ltrb 1 0 0 0>}}
  }
  GenerateAnimswhenAddCbox {
   widgetClassPath=&EditorUtilityCheckBox
   bIsVariable=true
   parent=HorizontalBox_82
   widget=GenerateAnimswhenAddCbox
   properties {checkedState=Checked, toolTipText="When unchecked, skip creating or updating the linked anims"}
   slot {class=HorizontalBoxSlot}
  }
  line_9 {
   widgetClassPath=&CommonBorder
   parent=HorizontalBox_82
   widget=line_9
   properties {
    background {
     bIsDynamicallyLoaded=false
     drawAs=Box
     imageType=NoImage
     mirroring=NoMirror
     resourceName=None
     resourceObject=null
     tiling=NoTile
     imageSize=<18 18>
     margin=<ltrb 0.4 0.4 0.4 0.4>
     outlineSettings {
      bUseBrushTransparency=false
      roundingType=HalfHeightRadius
      width=-9.2626
      color {colorUseRule=UseColor_Specified, specifiedColor=<rgba 0 0 0 0>}
      cornerRadii=<0 0 0 0>
     }
     tintColor {colorUseRule=UseColor_Specified, specifiedColor=<rgba 0 0 0 1>}
     uVRegion {bIsValid=false, max=<0 0>, min=<0 0>}
    }
    padding=<ltrb 1.8 2 1.8 2>
    renderTransform {angle=0, scale=<0.6 1>, shear=<0 0>, translation=<0 0>}
   }
   slot {class=HorizontalBoxSlot, properties {padding=<ltrb 0 0 1 0>}}
  }
  LayeredMasterRigText {
   widgetClassPath=&TextBlock
   parent=HorizontalBox_82
   widget=LayeredMasterRigText
   properties {
    text=Layered Master Rig
    font {
     bForceMonospaced=false
     bMaterialIsStencil=false
     fontMaterial=null
     fontObject=null
     letterSpacing=0
     monospacedWidth=1
     size=10.5
     skewAmount=0
     typefaceFontName=Default
     outlineSettings {
      bApplyOutlineToDropShadows=false
      bMiteredCorners=false
      bSeparateFillAlpha=false
      outlineMaterial=null
      outlineSize=0
      outlineColor=<rgba 0 0 0 1>
     }
    }
   }
   slot {class=HorizontalBoxSlot, properties {padding=<ltrb 1 0 0 0>}}
  }
  LayeredMasterRigCbox {
   widgetClassPath=&EditorUtilityCheckBox
   bIsVariable=true
   parent=HorizontalBox_82
   widget=LayeredMasterRigCbox
   properties {checkedState=Checked, toolTipText="When Checked, The Master's Control Rig is set to layered"}
   slot {class=HorizontalBoxSlot}
  }
  Spacer_289 {widgetClassPath=&Spacer, parent=VerticalBox_0, widget=Spacer_289, slot {class=VerticalBoxSlot}}
  line_4 {
   widgetClassPath=&CommonBorder
   parent=VerticalBox_0
   widget=line_4
   properties {background {
     bIsDynamicallyLoaded=false
     drawAs=RoundedBox
     imageType=NoImage
     mirroring=NoMirror
     resourceName=None
     resourceObject=null
     tiling=NoTile
     imageSize=<18 18>
     margin=<ltrb 0.4 0.4 0.4 0.4>
     outlineSettings {
      bUseBrushTransparency=false
      roundingType=HalfHeightRadius
      width=1.4037
      color {colorUseRule=UseColor_Specified, specifiedColor=<rgba 0 0 0 0>}
      cornerRadii=<-0.4977 0.3303 0 -0.5778>
     }
     tintColor {colorUseRule=UseColor_Specified, specifiedColor=<rgba 0 0 0 1>}
     uVRegion {bIsValid=false, max=<0 0>, min=<0 0>}
    }}
   slot {class=VerticalBoxSlot}
  }
  SizeBox_224 {
   widgetClassPath=&SizeBox
   parent=VerticalBox_0
   widget=SizeBox_224
   properties {bOverride_HeightOverride=true, heightOverride=85.8151}
   slot {class=VerticalBoxSlot}
  }
  Border_41 {
   widgetClassPath=&Border
   parent=SizeBox_224
   widget=Border_41
   properties {
    background {
     bIsDynamicallyLoaded=false
     drawAs=Image
     imageType=NoImage
     mirroring=NoMirror
     resourceName=None
     resourceObject=null
     tiling=NoTile
     imageSize=<32 32>
     margin=<ltrb 0 0 0 0>
     outlineSettings {
      bUseBrushTransparency=false
      roundingType=HalfHeightRadius
      width=0
      color {colorUseRule=UseColor_Specified, specifiedColor=<rgba 0 0 0 0>}
      cornerRadii=<0 0 0 0>
     }
     tintColor {colorUseRule=UseColor_Specified, specifiedColor=<rgba 0.0254 0.0259 0.0522 1>}
     uVRegion {bIsValid=false, max=<0 0>, min=<0 0>}
    }
    padding=<ltrb 0 0 0 0>
   }
   slot {class=SizeBoxSlot}
  }
  CanvasPanel_271 {
   widgetClassPath=&CanvasPanel
   parent=Border_41
   widget=CanvasPanel_271
   slot {class=BorderSlot, properties {padding=<ltrb 0 0 0 0>}}
  }
  HorizontalBox_1 {
   widgetClassPath=&HorizontalBox
   parent=CanvasPanel_271
   widget=HorizontalBox_1
   slot {class=CanvasPanelSlot, properties {layoutData {alignment=<0 0>, anchors {maximum=<1 1>, minimum=<0 0>}, offsets=<ltrb 0 0 1 0>}}}
  }
  IncludeSourcecbox {
   widgetClassPath=&EditorUtilityCheckBox
   bIsVariable=true
   parent=HorizontalBox_1
   widget=IncludeSourcecbox
   properties {checkedState=Checked, toolTipText=Update the Source info and re-export Linked Anim. Always True when creating a new LS}
   slot {class=HorizontalBoxSlot}
  }
  ParentCIDView {
   widgetClassPath=&SinglePropertyView
   parent=HorizontalBox_1
   widget=ParentCIDView
   slot {class=HorizontalBoxSlot, properties {size {sizeRule=Fill, value=1}}}
  }
  line_5 {
   widgetClassPath=&CommonBorder
   parent=HorizontalBox_1
   widget=line_5
   properties {
    background {
     bIsDynamicallyLoaded=false
     drawAs=Box
     imageType=NoImage
     mirroring=NoMirror
     resourceName=None
     resourceObject=null
     tiling=NoTile
     imageSize=<18 18>
     margin=<ltrb 0.4 0.4 0.4 0.4>
     outlineSettings {
      bUseBrushTransparency=false
      roundingType=HalfHeightRadius
      width=-9.2626
      color {colorUseRule=UseColor_Specified, specifiedColor=<rgba 0 0 0 0>}
      cornerRadii=<0 0 0 0>
     }
     tintColor {colorUseRule=UseColor_Specified, specifiedColor=<rgba 0 0 0 1>}
     uVRegion {bIsValid=false, max=<0 0>, min=<0 0>}
    }
    padding=<ltrb 1 2 1 2>
   }
   slot {class=HorizontalBoxSlot}
  }
  Spacer_197 {
   widgetClassPath=&Spacer
   parent=HorizontalBox_1
   widget=Spacer_197
   properties {renderTransform {angle=0, scale=<1 2>, shear=<0 0>, translation=<0 0>}, size=<2 1>}
   slot {class=HorizontalBoxSlot}
  }
  ParentCrigView {
   widgetClassPath=&SinglePropertyView
   parent=HorizontalBox_1
   widget=ParentCrigView
   slot {class=HorizontalBoxSlot, properties {size {sizeRule=Fill, value=1}}}
  }
  Spacer_288 {widgetClassPath=&Spacer, parent=HorizontalBox_1, widget=Spacer_288, properties {size=<2 1>}, slot {class=HorizontalBoxSlot}}
  line {
   widgetClassPath=&CommonBorder
   parent=HorizontalBox_1
   widget=line
   properties {
    background {
     bIsDynamicallyLoaded=false
     drawAs=Box
     imageType=NoImage
     mirroring=NoMirror
     resourceName=None
     resourceObject=null
     tiling=NoTile
     imageSize=<18 18>
     margin=<ltrb 0.4 0.4 0.4 0.4>
     outlineSettings {
      bUseBrushTransparency=false
      roundingType=HalfHeightRadius
      width=-9.2626
      color {colorUseRule=UseColor_Specified, specifiedColor=<rgba 0 0 0 0>}
      cornerRadii=<0 0 0 0>
     }
     tintColor {colorUseRule=UseColor_Specified, specifiedColor=<rgba 0 0 0 1>}
     uVRegion {bIsValid=false, max=<0 0>, min=<0 0>}
    }
    padding=<ltrb 1 2 1 2>
   }
   slot {class=HorizontalBoxSlot}
  }
  ParentAnimationSinglePropertyView {
   widgetClassPath=&SinglePropertyView
   parent=HorizontalBox_1
   widget=ParentAnimationSinglePropertyView
   slot {class=HorizontalBoxSlot, properties {size {sizeRule=Fill, value=1}}}
  }
  SectionNameText {
   widgetClassPath=&TextBlock
   parent=CanvasPanel_271
   widget=SectionNameText
   properties {
    text=Source Setup
    visibility=SelfHitTestInvisible
    colorAndOpacity {colorUseRule=UseColor_Specified, specifiedColor=<rgba 0.5255 0.5255 0.5255 1>}
    font {
     bForceMonospaced=false
     bMaterialIsStencil=false
     fontMaterial=null
     fontObject=&Roboto
     letterSpacing=0
     monospacedWidth=1
     size=9
     skewAmount=0
     typefaceFontName=Bold
     outlineSettings {
      bApplyOutlineToDropShadows=false
      bMiteredCorners=false
      bSeparateFillAlpha=false
      outlineMaterial=null
      outlineSize=0
      outlineColor=<rgba 0 0 0 1>
     }
    }
    renderTransform {angle=0, scale=<1 1>, shear=<0 0>, translation=<0 2>}
    strikeBrush {
     bIsDynamicallyLoaded=false
     drawAs=Image
     imageType=NoImage
     mirroring=NoMirror
     resourceName=None
     resourceObject=null
     tiling=NoTile
     imageSize=<32 32>
     margin=<ltrb 0 0 0 0>
     outlineSettings {
      bUseBrushTransparency=false
      roundingType=HalfHeightRadius
      width=0
      color {colorUseRule=UseColor_Specified, specifiedColor=<rgba 0 0 0 0>}
      cornerRadii=<0 0 0 0>
     }
     tintColor {colorUseRule=UseColor_Specified, specifiedColor=<rgba 1 1 1 0.0957>}
     uVRegion {bIsValid=false, max=<0 0>, min=<0 0>}
    }
   }
   slot {
    class=CanvasPanelSlot
    properties {layoutData {alignment=<-0.0081 0.0243>, anchors {maximum=<0 0>, minimum=<0 0>}, offsets=<ltrb 0 0 100 30>}}
   }
  }
  line_2 {
   widgetClassPath=&CommonBorder
   parent=VerticalBox_0
   widget=line_2
   properties {background {
     bIsDynamicallyLoaded=false
     drawAs=RoundedBox
     imageType=NoImage
     mirroring=NoMirror
     resourceName=None
     resourceObject=null
     tiling=NoTile
     imageSize=<18 18>
     margin=<ltrb 0.4 0.4 0.4 0.4>
     outlineSettings {
      bUseBrushTransparency=false
      roundingType=HalfHeightRadius
      width=1.4037
      color {colorUseRule=UseColor_Specified, specifiedColor=<rgba 0 0 0 0>}
      cornerRadii=<-0.4977 0.3303 0 -0.5778>
     }
     tintColor {colorUseRule=UseColor_Specified, specifiedColor=<rgba 0 0 0 1>}
     uVRegion {bIsValid=false, max=<0 0>, min=<0 0>}
    }}
   slot {class=VerticalBoxSlot}
  }
  Spacer_260 {widgetClassPath=&Spacer, parent=VerticalBox_0, widget=Spacer_260, properties {size=<1 1.2635>}, slot {class=VerticalBoxSlot}}
  SizeBox_297 {
   widgetClassPath=&SizeBox
   parent=VerticalBox_0
   widget=SizeBox_297
   properties {bOverride_HeightOverride=true, heightOverride=363.1833}
   slot {class=VerticalBoxSlot}
  }
  Border_387 {
   widgetClassPath=&Border
   parent=SizeBox_297
   widget=Border_387
   properties {brushColor=<rgba 0.02 0.0185 0.0183 1>, padding=<ltrb 0 0 0 0>}
   slot {class=SizeBoxSlot}
  }
  VerticalBox_124 {
   widgetClassPath=&VerticalBox
   parent=Border_387
   widget=VerticalBox_124
   properties {clipping=ClipToBoundsAlways}
   slot {class=BorderSlot, properties {padding=<ltrb 0 0 0 0>}}
  }
  HorizontalBox_111 {widgetClassPath=&HorizontalBox, parent=VerticalBox_124, widget=HorizontalBox_111, slot {class=VerticalBoxSlot}}
  TextBlock_65 {
   widgetClassPath=&TextBlock
   parent=HorizontalBox_111
   widget=TextBlock_65
   properties {
    text=Targets
    visibility=HitTestInvisible
    colorAndOpacity {colorUseRule=UseColor_Specified, specifiedColor=<rgba 0.5255 0.5255 0.5255 1>}
    font {
     bForceMonospaced=false
     bMaterialIsStencil=false
     fontMaterial=null
     fontObject=&Roboto
     letterSpacing=0
     monospacedWidth=1
     size=10.5
     skewAmount=0
     typefaceFontName=Bold
     outlineSettings {
      bApplyOutlineToDropShadows=false
      bMiteredCorners=false
      bSeparateFillAlpha=false
      outlineMaterial=null
      outlineSize=0
      outlineColor=<rgba 0 0 0 1>
     }
    }
    strikeBrush {
     bIsDynamicallyLoaded=false
     drawAs=Image
     imageType=NoImage
     mirroring=NoMirror
     resourceName=None
     resourceObject=null
     tiling=NoTile
     imageSize=<32 32>
     margin=<ltrb 0 0 0 0>
     outlineSettings {
      bUseBrushTransparency=false
      roundingType=HalfHeightRadius
      width=0
      color {colorUseRule=UseColor_Specified, specifiedColor=<rgba 0 0 0 0>}
      cornerRadii=<0 0 0 0>
     }
     tintColor {colorUseRule=UseColor_Specified, specifiedColor=<rgba 1 1 1 0.0957>}
     uVRegion {bIsValid=false, max=<0 0>, min=<0 0>}
    }
   }
   slot {class=HorizontalBoxSlot, properties {verticalAlignment=VAlign_Center}}
  }
  line_7 {
   widgetClassPath=&CommonBorder
   parent=HorizontalBox_111
   widget=line_7
   properties {background {
     bIsDynamicallyLoaded=false
     drawAs=RoundedBox
     imageType=NoImage
     mirroring=NoMirror
     resourceName=None
     resourceObject=null
     tiling=NoTile
     imageSize=<18 18>
     margin=<ltrb 0.4 0.4 0.4 0.4>
     outlineSettings {
      bUseBrushTransparency=false
      roundingType=HalfHeightRadius
      width=1.4037
      color {colorUseRule=UseColor_Specified, specifiedColor=<rgba 0 0 0 0>}
      cornerRadii=<-0.4977 0.3303 0 -0.5778>
     }
     tintColor {colorUseRule=UseColor_Specified, specifiedColor=<rgba 0 0 0 1>}
     uVRegion {bIsValid=false, max=<0 0>, min=<0 0>}
    }}
   slot {class=HorizontalBoxSlot, properties {verticalAlignment=VAlign_Center, size {sizeRule=Fill, value=1}}}
  }
  line_6 {
   widgetClassPath=&CommonBorder
   parent=HorizontalBox_111
   widget=line_6
   properties {background {
     bIsDynamicallyLoaded=false
     drawAs=RoundedBox
     imageType=NoImage
     mirroring=NoMirror
     resourceName=None
     resourceObject=null
     tiling=NoTile
     imageSize=<18 18>
     margin=<ltrb 0.4 0.4 0.4 0.4>
     outlineSettings {
      bUseBrushTransparency=false
      roundingType=HalfHeightRadius
      width=1.4037
      color {colorUseRule=UseColor_Specified, specifiedColor=<rgba 0 0 0 0>}
      cornerRadii=<-0.4977 0.3303 0 -0.5778>
     }
     tintColor {colorUseRule=UseColor_Specified, specifiedColor=<rgba 0 0 0 1>}
     uVRegion {bIsValid=false, max=<0 0>, min=<0 0>}
    }}
   slot {class=HorizontalBoxSlot, properties {verticalAlignment=VAlign_Center, size {sizeRule=Fill, value=1}}}
  }
  line_3 {
   widgetClassPath=&CommonBorder
   parent=VerticalBox_124
   widget=line_3
   properties {
    background {
     bIsDynamicallyLoaded=false
     drawAs=RoundedBox
     imageType=NoImage
     mirroring=NoMirror
     resourceName=None
     resourceObject=null
     tiling=NoTile
     imageSize=<18 18>
     margin=<ltrb 0.4 0.4 0.4 0.4>
     outlineSettings {
      bUseBrushTransparency=false
      roundingType=HalfHeightRadius
      width=1.4037
      color {colorUseRule=UseColor_Specified, specifiedColor=<rgba 0 0 0 0>}
      cornerRadii=<-0.4977 0.3303 0 -0.5778>
     }
     tintColor {colorUseRule=UseColor_Specified, specifiedColor=<rgba 0 0 0 1>}
     uVRegion {bIsValid=false, max=<0 0>, min=<0 0>}
    }
    renderTransform {angle=0, scale=<-1 1>, shear=<0 0>, translation=<-12 1.1>}
   }
   slot {class=VerticalBoxSlot}
  }
  TargetWidgetsScrollBox {
   widgetClassPath=&EditorUtilityScrollBox
   bIsVariable=true
   parent=VerticalBox_124
   widget=TargetWidgetsScrollBox
   properties {alwaysShowScrollbar=true}
   slot {class=VerticalBoxSlot, properties {size {sizeRule=Fill, value=1}}}
  }
  Overlay_278 {
   widgetClassPath=&Overlay
   parent=CanvasPanel_33
   widget=Overlay_278
   slot {class=CanvasPanelSlot, properties {layoutData {alignment=<0 0>, anchors {maximum=<1 1>, minimum=<0 0>}, offsets=<ltrb 0 0 0 608.1081>}}}
  }
  CommonBorder_33 {
   widgetClassPath=&CommonBorder
   parent=Overlay_278
   widget=CommonBorder_33
   properties {background {
     bIsDynamicallyLoaded=false
     drawAs=NoDrawType
     imageType=NoImage
     mirroring=NoMirror
     resourceName=None
     resourceObject=null
     tiling=NoTile
     imageSize=<32 32>
     margin=<ltrb 0 0 0 0>
     outlineSettings {
      bUseBrushTransparency=false
      roundingType=HalfHeightRadius
      width=0
      color {colorUseRule=UseColor_Specified, specifiedColor=<rgba 0 0 0 0>}
      cornerRadii=<0 0 0 0>
     }
     tintColor {colorUseRule=UseColor_Specified, specifiedColor=<rgba 0.151 0.151 0.151 1>}
     uVRegion {bIsValid=false, max=<0 0>, min=<0 0>}
    }}
   slot {class=OverlaySlot, properties {horizontalAlignment=HAlign_Fill, verticalAlignment=VAlign_Fill}}
  }
  anim_dialog {widgetClassPath=&NamedSlot, bIsVariable=true, parent=CommonBorder_33, widget=anim_dialog, slot {class=BorderSlot}}
 ]
}
